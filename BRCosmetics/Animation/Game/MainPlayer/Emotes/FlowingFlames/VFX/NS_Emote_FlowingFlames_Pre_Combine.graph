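# Niagara system for the Flowing Flames emote, attaching flame meshes, fire sprites, sparks and flickering lights to two skeletal mesh bones.
# Most emitters sample bones via SkeletalMeshLocation, several using a partition index to mirror directions along each bone's X axis.
# A collision emitter traces from both bones to position its particles, offset in front-end preview.
# A multi-ribbon fire-to-water trail, driven by user velocity and gravity parameters, is disabled.
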NiagaraSystem /BRCosmetics/Animation/Game/MainPlayer/Emotes/FlowingFlames/VFX/NS_Emote_FlowingFlames_Pre_Combine
dependencies {
 usedDataInterfaces [
  &NiagaraDataInterfaceCollisionQuery
  &NiagaraDataInterfaceDebugDraw
  &NiagaraDataInterfaceCurve
  &NiagaraDataInterfaceParticleRead
  &NiagaraDataInterfacePlatformSet
  &NiagaraDataInterfaceArrayInt32
  &NiagaraDataInterfaceColorCurve
  &NiagaraDataInterfaceVectorCurve
  &NiagaraDataInterfaceSkeletalMesh
 ]
 usedDynamicInputs [&VectorFromCurve, &SimulationPosition]
 usedModules [
  &self:SystemSpawnScript.SystemScriptSource.SystemScriptGraph.NiagaraNodeAssignment_0.SetVariables_#1
  &SystemState
 ]
}
emitters {
 Collision_Position {
  emitterUpdateScript [
   &EmitterState()
   &SpawnBurst_Instantaneous(Spawn Count=2, Use Spawn Probability=false, UseLoopCountLimit=false)
  ]
  particleSpawnScript [
   &InitializeParticle(Color Mode=::NewEnumerator0, Sprite Size Mode=::NewEnumerator3, UsePositionOffset=false, Write Material Random=true)
   &PartitionParticles()
  ]
  particleUpdateScript [
   &ParticleState(Kill Particles When Lifetime Has Elapsed=false)
   &SkeletalMeshLocation(
    Bone / Socket Index=$Particles.PartitionParticles.Partition
    Bone Sampling Mode=::NewEnumerator4
    MaskSpawnGroup=false
    Orientation Sampling=::NewEnumerator1
    Position Sampling=::NewEnumerator1
    Skeletal Mesh=$System.SkeletalMeshUser
   )
   &self:GridCollision(
    StartPos:NiagaraPosition=&MergeHelper_VectorSwitch(
     0=Linked() {linkedVariable {name=Output.SkeletalMeshLocation.SampledPositionWithoutOffset, type=NiagaraPosition}}
     1=Linked() {linkedVariable {name=Output.SkeletalMeshLocation001.SampledPositionWithoutOffset, type=NiagaraPosition}}
     2=?
     3=?
     Index=$Particles.PartitionParticles.Partition
     Num=::NewEnumerator0
    )
    TraceDir:Vector3f=&MergeHelper_VectorSwitch(
     0=&Multiply_VectorByFloat(
      Vector=&ReturnMeshOrientationAxis(Mesh Orientation Axis=::XAxis, Mesh Orientation Quat=$Particles.SkeletalMeshLocation.SampledRotation)
      Float=-1
     )
     1=&ReturnMeshOrientationAxis(Mesh Orientation Axis=::XAxis, Mesh Orientation Quat=$Particles.SkeletalMeshLocation001.SampledRotation)
     2=?
     3=?
     Index=$Particles.PartitionParticles.Partition
     Num=::NewEnumerator0
    )
    TraceDist=100
    bIsFrontEnd:NiagaraBool=$User.bIsFrontEndPreview
   ) {module=GridCollision}
   /SetVariables(Particles.Position=&FN_FrontEndVector3Modifier(
     Gameplay=Linked() {linkedVariable {name=Particles.CollisionPos, type=NiagaraPosition}}
     Front End=&AddVectorToPosition(
      Position=$Particles.CollisionPos
      Vector=&MergeHelper_VectorSwitch(
       0=&Multiply_VectorByFloat(
        Vector=&ReturnMeshOrientationAxis(Mesh Orientation Axis=::XAxis, Mesh Orientation Quat=$Particles.SkeletalMeshLocation.SampledRotation)
        Float=-50
       )
       1=&Multiply_VectorByFloat(
        Vector=&ReturnMeshOrientationAxis(Mesh Orientation Axis=::XAxis, Mesh Orientation Quat=$Particles.SkeletalMeshLocation.SampledRotation)
        Float=50
       )
       2=?
       3=?
       Index=$Particles.PartitionParticles.Partition
       Num=::NewEnumerator0
      )
     )
    )) {isSetParameters=true}
  ]
  rendererClasses [&NiagaraSpriteRendererProperties]
  renderers [{rendererClass=&NiagaraSpriteRendererProperties}]
 }
 Comet_Flame_Mesh {
  emitterUpdateScript [
   &EmitterState(Life Cycle Mode=::NewEnumerator1, Loop Behavior=::NewEnumerator1, Loop Delay=0.2, Loop Duration Mode=::NewEnumerator1, UseLoopDelay=false)
   &SpawnRate(SpawnRate=20, Use Spawn Probability=false)
  ]
  particleSpawnScript [
   &InitializeParticle(
    Color=<rgba 1.5 0.375 0.15 5>
    Lifetime=0.2
    Mass Mode=::NewEnumerator1
    Mesh Scale=<-0.04 0.035 0.035>
    Mesh Scale Mode=::NewEnumerator4
    UsePositionOffset=false
    Write Material Random=true
   ) {module=InitializeParticle001}
   &PartitionParticles()
   &InitialMeshOrientation(Mesh Axis To Orient=<0.5 0 0>, Mesh Orientation Mode=::NewEnumerator2, Orientation Vector=<0.1 0 0>, Use Rotation Vector=false)
   /SetVariables(Particles.Color_2:LinearColor=<rgba 20 12.5 2 0.25>, Particles.Scale_2:Vector3f=<0.035 0.025 0.025>) {isSetParameters=true}
   &DynamicMaterialParameters(
    Index 0 Param 1=5
    Index 0 Param 2=5
    Module.Index0Float0Write=true
    Module.Index0Float1Write=true
    Module.Index0Float2Write=true
    Module.Index0Float3Write=true
   )
   /SetVariables(Particles.PreviousPosition:NiagaraPosition=$Engine.Emitter.SimulationPosition) {isSetParameters=true}
  ]
  particleUpdateScript [
   &SkeletalMeshLocation(
    Bone / Socket Index=$Particles.PartitionParticles.Partition
    Bone Sampling Mode=::NewEnumerator4
    MaskSpawnGroup=false
    Orientation Sampling=::NewEnumerator1
    Skeletal Mesh=$System.SkeletalMeshUser
   )
   /SetVariables(Particles.CurrentPosition:NiagaraPosition=$Particles.SkeletalMeshLocation.SampledPosition) {isSetParameters=true}
   /SetVariables(Particles.OrientationVector:Vector3f=&Add_Vector(A=&Subtract_Position(A=$Particles.PreviousPosition, B=$Particles.CurrentPosition), B=<0 0 -5>)) {isSetParameters=true}
   /SetVariables(Particles.PreviousPosition:NiagaraPosition=&self:ScratchDynamicInput(Position:NiagaraPosition=$Particles.PreviousPosition, Newvalue:NiagaraPosition=$Particles.CurrentPosition)) {isSetParameters=true}
   &ParticleState()
   &UpdateMeshOrientation(
    Facing Direction=$Particles.OrientationVector
    Facing Interpolation=false
    Orientation Method=::NewEnumerator1
    Override Quat=false
    TransformFacing=false
    TransformUp=false
   ) {module=UpdateMeshOrientation001}
   &UpdateMeshOrientation(
    Facing Direction=&Lerp_Vector(
     A=&MergeHelper_VectorSwitch(
      0=&AddVectorToPosition(
       Position=&Multiply_VectorByFloat(
        Vector=&ReturnMeshOrientationAxis(Mesh Orientation Axis=::XAxis, Mesh Orientation Quat=$Particles.SkeletalMeshLocation.SampledRotation)
        Float=-1
       )
       Vector=<0 0 0>
      )
      1=&AddVectorToPosition(
       Position=&Multiply_VectorByFloat(
        Vector=&ReturnMeshOrientationAxis(Mesh Orientation Axis=::XAxis, Mesh Orientation Quat=$Particles.SkeletalMeshLocation.SampledRotation)
        Float=1
       )
       Vector=<0 0 0>
      )
      2=?
      3=?
      Index=$Particles.PartitionParticles.Partition
      Num=::NewEnumerator0
     )
     B=&Multiply_VectorByFloat(Vector=$Particles.SkeletalMeshLocation.SampledVelocity, Float=-1)
     Alpha=&FN_FrontEndFloatModifier(Gameplay=&Multiply_Float(A=$User.C_Inherit_Velocity, B=1), Front End=$User.C_Inherit_Velocity)
    )
    Facing Interpolation=false
    Orientation Method=::NewEnumerator1
    Override Quat=false
    Stabilization Mode:ENiagara_OrientVectorCount=::NewEnumerator2
    TransformFacing=false
    TransformSide=false
   ) {enabled=false}
   &ScaleColor(
    Scale Alpha=&Lerp_Float(A=1, B=0, Alpha=$Particles.NormalizedAge)
    Scale RGB=&MakeVectorFromLinearColorRGB(Color=&ColorFromCurve(ColorCurve=DataInterface() {propertyValues {
        AlphaCurve {keys []}
        BlueCurve {keys [<at 0 0.25 RCIM_Cubic>,<at 0.492 0.5>,<at 1 0.875 RCIM_Cubic>]}
        GreenCurve {keys [<at 0 0.4375 RCIM_Cubic>,<at 0.492 0.625>,<at 1 0.9062 RCIM_Cubic>]}
        RedCurve {keys [<at 0 1 RCIM_Cubic>,<at 0.492 1>,<at 1 1 RCIM_Cubic>]}
       }}))
    ScaleA=true
    ScaleRGB=true
   )
   &ScaleMeshSize(Scale Factor=&MakeVector(
     X=&Multiply_Float(
      A=$User.Flame_Size
      B=&FloatFromCurve(FloatCurve=DataInterface() {propertyValues {Curve {keys [<at 0 0.75 RCIM_Cubic>, {arriveTangent=-0.25, interpMode=RCIM_Cubic, leaveTangent=-0.25, time=0.5, value=2.5}, <at 1 0.5 RCIM_Cubic>]}}})
     )
     Y=&Multiply_Float(
      A=$User.Flame_Size
      B=&FloatFromCurve(FloatCurve=DataInterface() {propertyValues {Curve {keys [<at 0.5 1 RCIM_Cubic>,<at 1 1.75 RCIM_Cubic>]}}})
     )
     Z=&Multiply_Float(
      A=$User.Flame_Size
      B=&FloatFromCurve(FloatCurve=DataInterface() {propertyValues {Curve {keys [<at 0.5 1 RCIM_Cubic>,<at 1 1.75 RCIM_Cubic>]}}})
     )
    ))
   /SetVariables(Particles.Scale_2:Vector3f=&Multiply_Vector(
     A=&Multiply_Vector(
      A=$Particles.Initial.Scale_2
      B=&MakeVector(
       X=&FloatFromCurve(
        FloatCurve=DataInterface() {propertyValues {Curve {keys [<at 0 0.75 RCIM_Cubic>, {arriveTangent=-0.25, interpMode=RCIM_Cubic, leaveTangent=-0.25, time=0.5, value=2.5}, <at 1 0.5 RCIM_Cubic>]}}}
        Scale Curve=$User.Flame_Size
       )
       Y=&FloatFromCurve(FloatCurve=DataInterface() {propertyValues {Curve {keys [<at 0.5 1 RCIM_Cubic>,<at 1 1.75 RCIM_Cubic>]}}}, Scale Curve=$User.Flame_Size)
       Z=&FloatFromCurve(FloatCurve=DataInterface() {propertyValues {Curve {keys [<at 0.5 1 RCIM_Cubic>,<at 1 1.75 RCIM_Cubic>]}}}, Scale Curve=$User.Flame_Size)
      )
     )
     B=<1.5 3 2>
    )) {isSetParameters=true}
   &ScaleMeshSize(Scale Factor=<1.5 3 2>) {module=ScaleMeshSize001}
  ]
  rendererClasses [&NiagaraMeshRendererProperties]
  renderers [
   {rendererClass=&NiagaraMeshRendererProperties, rendererIndex=0}
   {rendererClass=&NiagaraMeshRendererProperties, rendererIndex=1}
  ]
 }
 ConeLights {
  emitterUpdateScript [
   &EmitterState(Life Cycle Mode=::NewEnumerator1, Loop Behavior=::NewEnumerator1, Loop Delay=1.7, Loop Duration=8, UseLoopDelay=false)
   &SpawnBurst_Instantaneous(Spawn Count=2, Use Spawn Probability=false, UseLoopCountLimit=false)
  ]
  particleSpawnScript [
   &InitializeParticle(
    Color=<rgba 1 0.2868 0 0.05>
    Lifetime=8
    Mesh Scale=<0.6 0.6 0.6>
    Mesh Scale Mode=::NewEnumerator4
    Sprite Size Mode=::NewEnumerator3
    Uniform Sprite Size=1000
    UsePositionOffset=false
    Write Material Random=true
   )
   &PartitionParticles()
   &InitialMeshOrientation(Mesh Orientation Mode=::NewEnumerator2, Use Rotation Vector=true)
  ]
  particleUpdateScript [
   &SolveForcesAndVelocity(Clamp Velocity=false, Limit Acceleration=false)
   &SkeletalMeshLocation(
    Bone / Socket Index=$Particles.PartitionParticles.Partition
    Bone Sampling Mode=::NewEnumerator4
    MaskSpawnGroup=false
    Orientation Sampling=::NewEnumerator1
    Position Sampling=::NewEnumerator1
    Skeletal Mesh=$System.SkeletalMeshUser
   )
   &ParticleState(Kill Particles When Lifetime Has Elapsed=false)
   &ScaleMeshSize(Scale Factor=&Multiply_Vector(
     A=&MakeVector(
      X=&FloatFromCurve(FloatCurve=DataInterface() {propertyValues {Curve {keys [
           <at 0 0 RCIM_Cubic>
           {arriveTangentWeight=0.0247, interpMode=RCIM_Cubic, leaveTangentWeight=0.0255, tangentMode=RCTM_User, time=0.0174, value=0.05}
           {arriveTangentWeight=0.0589, interpMode=RCIM_Cubic, leaveTangentWeight=0.2882, tangentMode=RCTM_User, time=0.0565, value=0.05}
           {
            arriveTangent=3.9696
            arriveTangentWeight=0.0453
            interpMode=RCIM_Cubic
            leaveTangent=3.9696
            leaveTangentWeight=0.4261
            tangentMode=RCTM_User
            time=0.1319
            value=0.75
           }
           {arriveTangent=0.288, interpMode=RCIM_Cubic, leaveTangent=0.288, time=0.2388, value=0.8}
           {arriveTangentWeight=0.3013, interpMode=RCIM_Cubic, tangentMode=RCTM_User, time=1, value=1}
          ]}}})
      Y=&FloatFromCurve(FloatCurve=DataInterface() {propertyValues {Curve {keys [
           <at 0 0 RCIM_Cubic>
           {arriveTangentWeight=0.0247, interpMode=RCIM_Cubic, leaveTangentWeight=0.0255, tangentMode=RCTM_User, time=0.0174, value=0.05}
           {arriveTangentWeight=0.0589, interpMode=RCIM_Cubic, leaveTangentWeight=0.2882, tangentMode=RCTM_User, time=0.0565, value=0.05}
           {
            arriveTangent=3.9696
            arriveTangentWeight=0.0453
            interpMode=RCIM_Cubic
            leaveTangent=3.9696
            leaveTangentWeight=0.4261
            tangentMode=RCTM_User
            time=0.1319
            value=0.75
           }
           {arriveTangent=0.288, interpMode=RCIM_Cubic, leaveTangent=0.288, time=0.2254, value=0.8}
           {arriveTangentWeight=0.3013, interpMode=RCIM_Cubic, tangentMode=RCTM_User, time=1, value=1}
          ]}}})
      Z=&Multiply_Float(
       A=&FloatFromCurve(FloatCurve=DataInterface() {propertyValues {Curve {keys [
            {arriveTangent=779.4893, interpMode=RCIM_Cubic, leaveTangent=779.4893, leaveTangentWeight=3.3909, tangentMode=RCTM_User, time=0, value=0}
            {
             arriveTangent=19.032
             arriveTangentWeight=2.8007
             interpMode=RCIM_Cubic
             leaveTangent=19.032
             leaveTangentWeight=13.8693
             tangentMode=RCTM_User
             time=0.099
             value=30
            }
            {arriveTangentWeight=0.2773, interpMode=RCIM_Cubic, leaveTangentWeight=0.3745, tangentMode=RCTM_User, time=1.2045, value=45}
           ]}}})
       B=&RemapRange(
        Input Value To Remap=&VectorLength(Vector=&Subtract_Position(
          A=Linked() {linkedVariable {name=Particles.SkeletalMeshLocation001.SampledScale, type=Vector3f}}
          B=Linked() {linkedVariable {name=Particles.SkeletalMeshLocation002.SampledScale, type=Vector3f}}
         ))
        Input Min=0
        Input Max=100
        Output Min=0
        Output Max=2.5
       )
      )
     )
     B=&Divide_Vector(A:Vector3f=<1 1 1>, B:Vector3f=$Engine.Owner.Scale)
    ))
   &ScaleColor(
    Scale Alpha=&FloatFromCurve(
     FloatCurve=DataInterface() {propertyValues {Curve {keys [
         {interpMode=RCIM_Cubic, leaveTangent=0.0402, tangentMode=RCTM_Break, time=0, value=0}
         <at 0.3 0.0075>
         {arriveTangent=-0.3583, leaveTangent=-0.3583, time=3.9058, value=0.0075}
         <at 3.9074 0.0005>
         <at 3.9555 0.0005>
         <at 3.9764 0.0075>
         <at 4.1455 0.0075>
         <at 4.1461 0.0005>
         <at 4.2511 0.0005>
         <at 4.2699 0.0075>
         <at 4.3362 0.0075>
         <at 4.3431 0.0005>
         <at 4.4684 0.0005>
         <at 4.497 0.0075>
        ]}}}
     CurveIndex=$System.Age
    )
    ScaleA=true
    ScaleRGB=true
   )
   &FN_Light_Attributes(
    Color=$Particles.Color
    Create Color Attribute=true
    Create Exponent Attribute=true
    Create Radius Attribute=true
    Exponent=4
    Radius=100
    Scale Color=&FloatFromCurve(FloatCurve=DataInterface() {propertyValues {Curve {keys [<at 0 0>,<at 0.02 1>,<at 1 1>]}}}, Scale Curve=10)
   )
   /SetVariables(Particles.DynamicMaterialParameterTechBeam:Vector4f=/Niagara/DynamicInputs/Multiply/Multiply_Vector4_ByFloat(
     Vector4:Vector4f=<0.8 0.8 1 1>
     Float:NiagaraFloat=&FloatFromCurve(
      FloatCurve=DataInterface() {propertyValues {Curve {keys [
          {interpMode=RCIM_Cubic, leaveTangent=1.6164, tangentMode=RCTM_Break, time=0, value=0}
          {arriveTangentWeight=0.173, interpMode=RCIM_Cubic, leaveTangentWeight=1.0013, tangentMode=RCTM_User, time=0.4503, value=1}
          {arriveTangent=-0.0121, leaveTangent=-0.3583, tangentMode=RCTM_User, time=4.0562, value=1}
          <at 4.0574 0.25>
          <at 4.1055 0.25>
          <at 4.1267 1>
          <at 4.2958 1>
          <at 4.2961 0.25>
          <at 4.4011 0.25>
          <at 4.4202 1>
          <at 4.4865 1>
          <at 4.4931 0.25>
          <at 4.6184 0.25>
          <at 4.6473 1>
         ]}}}
      CurveIndex=$Engine.System.Age
     )
    )) {isSetParameters=true}
   &DynamicMaterialParameters(
    Index 0 Param 4=&FloatFromCurve(
     FloatCurve=DataInterface() {propertyValues {Curve {keys [
         {interpMode=RCIM_Cubic, leaveTangent=1.6164, tangentMode=RCTM_Break, time=0, value=0}
         {arriveTangentWeight=0.173, interpMode=RCIM_Cubic, leaveTangentWeight=1.0013, tangentMode=RCTM_User, time=0.4503, value=1}
         {arriveTangent=-0.0121, leaveTangent=-0.3583, tangentMode=RCTM_User, time=4.0562, value=1}
         <at 4.0574 0.25>
         <at 4.1055 0.25>
         <at 4.1267 1>
         <at 4.2958 1>
         <at 4.2961 0.25>
         <at 4.4011 0.25>
         <at 4.4202 1>
         <at 4.4865 1>
         <at 4.4931 0.25>
         <at 4.6184 0.25>
         <at 4.6473 1>
        ]}}}
     CurveIndex=$Engine.System.Age
    )
    Index 0 Vector 1:Vector3f=&FN_Select_Vector_from_Time_Of_Day(
     Morning=&MakeVectorFromLinearColorRGB(Color=<rgba 3 2.4842 0.7221 1>)
     Day=&MakeVectorFromLinearColorRGB(Color=<rgba 3 2.4842 0.7221 1>)
     Evening=&MakeVectorFromLinearColorRGB(Color=<rgba 3 2.4842 0.7221 1>)
     Night=&MakeVectorFromLinearColorRGB(Color=<rgba 5 2.6977 0.5729 1>)
    )
    Module.Index0Float3Write=true
    Module.Index0VectorWrite=true
    Write Parameter Index 0=::NewEnumerator3
   )
  ]
  rendererClasses [&NiagaraLightRendererProperties]
  renderers [{rendererClass=&NiagaraLightRendererProperties, rendererIndex=0}]
 }
 FireSprite001 {
  emitterUpdateScript [
   &EmitterState(Life Cycle Mode=::NewEnumerator1, Loop Duration=1, UseLoopDelay=true)
   &SpawnPerFrame(Use Spawn Probability=false)
   &SpawnRate(SpawnRate=45, Use Spawn Probability=false)
  ]
  particleSpawnScript [
   &InitializeParticle(
    Color=<rgba 1 0.884 0.3542 1>
    Lifetime Max=0.35
    Lifetime Min=0.35
    Lifetime Mode=::NewEnumerator1
    Position Offset=<0 0 35>
    Sprite Rotation Mode=::NewEnumerator2
    Sprite Size Mode=::NewEnumerator3
    Uniform Sprite Size=&Multiply_Float(
     A=&Lerp_Float(
      A=&RandomRangeFloat(Minimum=70, Maximum=90)
      B=&RandomRangeFloat(Minimum=90, Maximum=90)
      Alpha=0
     )
     B=0.7
    )
    UsePositionOffset=false
    Write Material Random=true
   )
   &SkeletalMeshLocation(
    Bone / Socket Index=&RandomRangeInt(Minimum=0, Maximum=1)
    Bone Sampling Mode=::NewEnumerator4
    MaskSpawnGroup=false
    Orientation Sampling=::NewEnumerator1
    Position Sampling=::NewEnumerator1
    Skeletal Mesh=$System.SkeletalMeshUser
   )
   &ShapeLocation(
    Cylinder Height=0
    Cylinder Height Midpoint=0
    Cylinder Radius=5
    MaskSpawnGroup=false
    Offset=<0 0 -8>
    Offset Mode=::NewEnumerator0
    Override Seed=false
    Shape Primitive=::NewEnumerator1
    Surface Only=true
    Surface Only Band Thickness=1
    UseOwnerScale=true
   )
   &AddVelocity(
    Module.Use Falloff Distance=false
    Origin Offset=<0 0 35>
    Override Seed=false
    Velocity Mode=::NewEnumerator1
    Velocity Speed=&RandomRangeFloat(Minimum=50, Maximum=150)
   ) {module=AddVelocity001}
   &CameraOffset(Camera Offset Amount=1500) {enabled=false}
   &CameraOffset(Camera Offset Amount=-4) {module=CameraOffset001}
  ]
  particleUpdateScript [
   &ParticleState()
   &ScaleColor(
    Scale Alpha=&Multiply_Float(
     A=&Lerp_Float(
      A=1
      B=0.4
      Alpha=&RemapRange(
       Input Value To Remap=&MakeFloatFromVector2d(Vector 2D=$Particles.SpriteSize)
       Input Min=50
       Input Max=70
       Output Min=0
       Output Max=1
      )
     )
     B=0.25
    )
    ScaleA=true
    ScaleRGB=true
   ) {module=ScaleColor002}
   &SpriteRotationRate(Rotation Rate=&RandomRangeFloat(Minimum=-70, Maximum=70))
   &ScaleSpriteSize(Uniform Curve Sprite Scale=DataInterface() {propertyValues {Curve {keys [{tangentMode=RCTM_User, time=0, value=1.5}, {tangentMode=RCTM_User, time=1, value=0.7}]}}})
   &Drag(
    Drag=&FloatFromCurve(FloatCurve=DataInterface() {propertyValues {Curve {keys [<at 0 0>,<at 0.5 5>]}}})
    Use Linear Drag=true
    Use Rotational Drag=true
   )
   &GravityForce(Gravity=&RandomRangeVector(Minimum=<0 0 200>, Maximum=<0 0 300>))
   &SolveForcesAndVelocity(Clamp Velocity=false, Limit Acceleration=false)
   &ScaleSpriteSizeBySpeed(Max Scale Factor=<2 5>, Min Scale Factor=<1 0.45>) {enabled=false}
   &ScaleColor(
    Linear Color Curve=DataInterface() {propertyValues {
      AlphaCurve {keys [<at 0 0>,<at 0.2511 0.5>,<at 1 0>]}
      BlueCurve {keys [<at 0 1>,<at 1 1>]}
      GreenCurve {keys [<at 0 1>,<at 1 0.1122>]}
      RedCurve {keys [<at 0 1>,<at 1 0.5193>]}
     }}
    Scale Mode=::NewEnumerator2
   )
   &DynamicMaterialParameters(
    Index 0 Param 1=&FloatFromCurve(FloatCurve=DataInterface() {propertyValues {Curve {keys [<at 0 0>,<at 1 1>]}}})
    Module.Index0Float0Write=true
    Module.Index0Float1Write=true
    Module.Index0Float2Write=true
    Module.Index0Float3Write=true
   )
   &ScaleColor(
    Linear Color Curve=DataInterface() {propertyValues {
      AlphaCurve {keys [{arriveTangent=-0.1596, interpMode=RCIM_Cubic, leaveTangent=-0.1596, tangentMode=RCTM_User, time=0.3182, value=1}, <at 0.982 0 RCIM_Cubic>]}
      BlueCurve {keys [
        <at 0.0505 3.4995 RCIM_Cubic>
        {arriveTangent=-3.8023, interpMode=RCIM_Cubic, leaveTangent=-3.8023, time=0.3055, value=1.1959}
        {arriveTangent=-1.0615, interpMode=RCIM_Cubic, leaveTangent=-1.0615, time=0.7879, value=0.6957}
        <at 0.9823 0.4774>
       ]}
      GreenCurve {keys [
        <at 0.0505 4.0716 RCIM_Cubic>
        {arriveTangent=-4.5324, interpMode=RCIM_Cubic, leaveTangent=-4.5324, time=0.3055, value=1.3562}
        {arriveTangent=-1.1361, interpMode=RCIM_Cubic, leaveTangent=-1.1361, time=0.7879, value=0.7294}
        <at 0.9823 0.5872>
       ]}
      RedCurve {keys [
        <at 0.0505 5 RCIM_Cubic>
        {arriveTangent=-5.4244, interpMode=RCIM_Cubic, leaveTangent=-5.4244, time=0.3055, value=2}
        {arriveTangent=-1.4775, interpMode=RCIM_Cubic, leaveTangent=-1.4775, time=0.7879, value=1}
        <at 0.9823 1>
       ]}
     }}
    Scale Mode=::NewEnumerator2
   ) {module=ScaleColor001}
   &ScaleColor(Scale Alpha=&FN_FrontEndFloatModifier(Gameplay=1, Front End=0.45), ScaleA=true, ScaleRGB=true) {module=ScaleColor003}
  ]
  rendererClasses [&NiagaraSpriteRendererProperties]
  renderers [{rendererClass=&NiagaraSpriteRendererProperties}]
 }
 Fire_Water_Trail {
  bEnabled=false
  emitterUpdateScript [
   &EmitterState()
   &SpawnRate(SpawnRate=100, Use Spawn Probability=false)
  ]
  particleSpawnScript [
   &InitializeParticle(Ribbon Width=100, Ribbon Width Mode=::NewEnumerator1, UsePositionOffset=false, Write Material Random=true)
   &PartitionParticles()
   &SkeletalMeshLocation(
    Bone / Socket Index=$Particles.PartitionParticles.Partition
    Bone Sampling Mode=::NewEnumerator4
    MaskSpawnGroup=false
    Orientation Sampling=::NewEnumerator1
    Position Sampling=::NewEnumerator1
    Skeletal Mesh=$System.SkeletalMeshUser
   )
   /SetVariables(Particles.Bright Fire Color:LinearColor=<rgba 0 9.1556 20 1>) {isSetParameters=true}
   /SetVariables(
    Particles.Bright Fire Width=50
    Particles.Fire Dark 1 Width=80
    Particles.Fire Dark 2 Width=80
    Particles.Transition Width=100
    Particles.Water Light 1 Width=50
    Particles.Water Light 2 Width=50
    Particles.Water Wake Width=50
    Particles.Water Width=50
   ) {isSetParameters=true}
   /SetVariables(Particles.Position=$Output.SkeletalMeshLocation.SampledPositionWithoutOffset) {isSetParameters=true, module=SetVariables_#2}
   &AddVelocity(
    Rotation Coordinate Space=::Simulation
    Velocity=&ReturnMeshOrientationAxis(Mesh Orientation Axis=::XAxis, Mesh Orientation Quat=$Particles.SkeletalMeshLocation.SampledRotation)
    Velocity Speed Scale=&MergeHelper_FloatSwitch(
     0=&Subtract_Float(A=0, B=$User.User_Velocity)
     1=$User.User_Velocity
     2=?
     3=?
     Index=$Particles.PartitionParticles.Partition
     Num=::NewEnumerator0
    )
   )
   &ApplyInitialForces()
  ]
  particleUpdateScript [
   &ParticleState()
   /SetVariables(
    Particles.Bright Fire Color:LinearColor=&MultiplyLinearColor(
     Linear Color=$Particles.Initial.Bright Fire Color
     Scale Factor=&MakeVector4(
      X=1
      Y=1
      Z=1
      W=&FloatFromCurve(
       FloatCurve=DataInterface() {propertyValues {Curve {
          defaultValue=1
          keys [
           {interpMode=RCIM_Cubic, tangentMode=RCTM_User, time=0, value=0}
           <at 0.1 1 RCIM_Cubic>
           {interpMode=RCIM_Cubic, leaveTangent=10, time=0.2, value=0}
          ]
         }}}
       CurveIndex=&Divide_Float(A=$Particles.NormalizedAge, B=0.6)
      )
     )
    )
    Particles.Fire Dark 1 Color:LinearColor=&MultiplyLinearColor(Linear Color=$Particles.Initial.Fire Dark 1 Color, Scale Factor=&MakeVector4(X=1, Y=1, Z=1, W=0.7))
    Particles.Fire Dark 2 Color:LinearColor=&MultiplyLinearColor(Linear Color=$Particles.Initial.Fire Dark 1 Color, Scale Factor=&MakeVector4(X=1, Y=1, Z=1, W=0.7))
    Particles.Transition Color:LinearColor=&MultiplyLinearColor(
     Linear Color=$Particles.Initial.Transition Color
     Scale Factor=&Multiply_Vector4(
      A=&MakeVector4_FromLinearColor(LinearColor=&ColorFromCurve(ColorCurve=DataInterface() {propertyValues {
          AlphaCurve {keys [<at 0.0207 0>,<at 0.1253 0.9793>,<at 0.1846 0>,<at 1 0>]}
          BlueCurve {keys [<at 0.003 0.3229>,<at 0.0491 0.4507>,<at 0.1611 0.3177>,<at 0.3437 2>]}
          GreenCurve {keys [<at 0.003 0.8537>,<at 0.0491 0.2376>,<at 0.1611 0.0757>,<at 0.3437 0.5232>]}
          RedCurve {keys [<at 0.003 1>,<at 0.0491 1>,<at 0.1611 0.1416>,<at 0.3437 0.0938>]}
         }}))
      B=<1 1 1 1.5>
     )
    )
    Particles.Water Color:LinearColor=&MultiplyLinearColor(
     Linear Color=$Particles.Initial.Water Color
     Scale Factor=&MakeVector4(
      X=1
      Y=1
      Z=1
      W=&FloatFromCurve(
       FloatCurve=DataInterface() {propertyValues {Curve {
          defaultValue=1
          keys [
           {interpMode=RCIM_Cubic, tangentMode=RCTM_User, time=0, value=0}
           {arriveTangent=2, interpMode=RCIM_Cubic, leaveTangent=2, time=0.2, value=0}
           <at 0.5 1 RCIM_Cubic>
           <at 1 0 RCIM_Cubic>
          ]
         }}}
       CurveIndex=&Divide_Float(A=$Particles.NormalizedAge, B=0.5)
       Scale Curve=0.3
      )
     )
    )
    Particles.Water Light 1 Color:LinearColor=&MultiplyLinearColor(
     Linear Color=$Particles.Initial.Water Light 1 Color
     Scale Factor=&MakeVector4(
      X=1
      Y=1
      Z=1
      W=&FloatFromCurve(
       FloatCurve=DataInterface() {propertyValues {Curve {
          defaultValue=1
          keys [
           {interpMode=RCIM_Cubic, tangentMode=RCTM_User, time=0, value=0}
           {arriveTangent=3.3333, interpMode=RCIM_Cubic, leaveTangent=3.3333, time=0.2, value=0}
           <at 0.3 1 RCIM_Cubic>
           <at 0.5 0 RCIM_Cubic>
          ]
         }}}
       CurveIndex=&Divide_Float(A=$Particles.NormalizedAge, B=0.5)
      )
     )
    )
    Particles.Water Light 2 Color:LinearColor=&MultiplyLinearColor(
     Linear Color=$Particles.Initial.Water Light 1 Color
     Scale Factor=&MakeVector4(
      X=1
      Y=1
      Z=1
      W=&FloatFromCurve(
       FloatCurve=DataInterface() {propertyValues {Curve {
          defaultValue=1
          keys [
           {interpMode=RCIM_Cubic, tangentMode=RCTM_User, time=0, value=0}
           {arriveTangent=3.3333, interpMode=RCIM_Cubic, leaveTangent=3.3333, time=0.2, value=0}
           <at 0.3 1 RCIM_Cubic>
           <at 0.5 0 RCIM_Cubic>
          ]
         }}}
       CurveIndex=&Divide_Float(A=$Particles.NormalizedAge, B=0.5)
      )
     )
    )
    Particles.Water Wake Color:LinearColor=&MultiplyLinearColor(
     Linear Color=$Particles.Initial.Water Color
     Scale Factor=&MakeVector4(
      X=1
      Y=1
      Z=1
      W=&FloatFromCurve(
       FloatCurve=DataInterface() {propertyValues {Curve {
          defaultValue=1
          keys [
           {interpMode=RCIM_Cubic, tangentMode=RCTM_User, time=0, value=0}
           {arriveTangent=2, interpMode=RCIM_Cubic, leaveTangent=2, time=0.2, value=0}
           <at 0.5 1 RCIM_Cubic>
           <at 1 0 RCIM_Cubic>
          ]
         }}}
       CurveIndex=&Divide_Float(A=$Particles.NormalizedAge, B=0.5)
       Scale Curve=0.3
      )
     )
    )
   ) {isSetParameters=true}
   /SetVariables(
    Particles.Bright Fire Width:NiagaraFloat=&Multiply_Float(
     A=$Particles.Initial.Bright Fire Width
     B=&FloatFromCurve(
      FloatCurve=DataInterface() {propertyValues {Curve {
         defaultValue=0
         keys [
          {interpMode=RCIM_Cubic, tangentMode=RCTM_User, time=0, value=1}
          {arriveTangent=-1.5474, interpMode=RCIM_Cubic, leaveTangent=-1.5474, tangentMode=RCTM_User, time=1, value=0}
         ]
        }}}
      CurveIndex=&Divide_Float(A=$Particles.NormalizedAge, B=0.6)
     )
    )
    Particles.Fire Dark 1 Width:NiagaraFloat=&Multiply_Float(
     A=$Particles.Initial.Fire Dark 1 Width
     B=&FloatFromCurve(
      FloatCurve=DataInterface() {propertyValues {Curve {
         defaultValue=0
         keys [
          {interpMode=RCIM_Cubic, tangentMode=RCTM_User, time=0, value=1}
          {arriveTangent=-1.5474, interpMode=RCIM_Cubic, leaveTangent=-1.5474, tangentMode=RCTM_User, time=1, value=0}
         ]
        }}}
      CurveIndex=&Divide_Float(A=$Particles.NormalizedAge, B=0.22)
     )
    )
    Particles.Fire Dark 2 Width:NiagaraFloat=&Multiply_Float(
     A=$Particles.Initial.Fire Dark 1 Width
     B=&FloatFromCurve(
      FloatCurve=DataInterface() {propertyValues {Curve {
         defaultValue=0
         keys [
          {interpMode=RCIM_Cubic, tangentMode=RCTM_User, time=0, value=1}
          {arriveTangent=-1.5474, interpMode=RCIM_Cubic, leaveTangent=-1.5474, tangentMode=RCTM_User, time=1, value=0}
         ]
        }}}
      CurveIndex=&Divide_Float(A=$Particles.NormalizedAge, B=0.22)
     )
    )
    Particles.Transition Width:NiagaraFloat=&Multiply_Float(
     A=$Particles.Initial.Transition Width
     B=&FloatFromCurve(FloatCurve=DataInterface() {propertyValues {Curve {
         defaultValue=0
         keys [
          {interpMode=RCIM_Cubic, tangentMode=RCTM_User, time=0, value=1}
          {arriveTangent=-1.5474, interpMode=RCIM_Cubic, leaveTangent=-1.5474, tangentMode=RCTM_User, time=1, value=0}
         ]
        }}})
    )
    Particles.Water Light 1 Width:NiagaraFloat=&Multiply_Float(
     A=$Particles.Initial.Water Light 1 Width
     B=&FloatFromCurve(
      FloatCurve=DataInterface() {propertyValues {Curve {
         defaultValue=0
         keys [
          {arriveTangent=-3.7011, interpMode=RCIM_Cubic, time=0, value=0}
          {arriveTangent=0.562, interpMode=RCIM_Cubic, leaveTangent=0.562, tangentMode=RCTM_User, time=0.3, value=1.1103}
          {interpMode=RCIM_Cubic, leaveTangent=0.3678, time=0.6, value=0}
         ]
        }}}
      CurveIndex=&Divide_Float(A=$Particles.NormalizedAge, B=0.5)
     )
    )
    Particles.Water Light 2 Width:NiagaraFloat=&Multiply_Float(
     A=$Particles.Initial.Water Light 2 Width
     B=&FloatFromCurve(
      FloatCurve=DataInterface() {propertyValues {Curve {
         defaultValue=0
         keys [
          {arriveTangent=-3.7011, interpMode=RCIM_Cubic, time=0, value=0}
          {arriveTangent=0.562, interpMode=RCIM_Cubic, leaveTangent=0.562, tangentMode=RCTM_User, time=0.3, value=1.1103}
          {interpMode=RCIM_Cubic, leaveTangent=0.3678, time=0.6, value=0}
         ]
        }}}
      CurveIndex=&Divide_Float(A=$Particles.NormalizedAge, B=0.5)
     )
    )
    Particles.Water Wake Width:NiagaraFloat=&Multiply_Float(
     A=$Particles.Initial.Water Wake Width
     B=&FloatFromCurve(
      FloatCurve=DataInterface() {propertyValues {Curve {
         defaultValue=0
         keys [
          {arriveTangent=-3.7011, interpMode=RCIM_Cubic, time=0, value=0}
          {arriveTangent=0.562, interpMode=RCIM_Cubic, leaveTangent=0.562, tangentMode=RCTM_User, time=0.3, value=1.1103}
          {interpMode=RCIM_Cubic, leaveTangent=0.3678, time=0.6, value=0}
         ]
        }}}
      CurveIndex=&Divide_Float(A=$Particles.NormalizedAge, B=0.5)
     )
    )
    Particles.Water Width:NiagaraFloat=&Multiply_Float(
     A=$Particles.Initial.Water Width
     B=&FloatFromCurve(
      FloatCurve=DataInterface() {propertyValues {Curve {
         defaultValue=0
         keys [
          {arriveTangent=-3.7011, interpMode=RCIM_Cubic, time=0, value=0}
          {arriveTangent=0.562, interpMode=RCIM_Cubic, leaveTangent=0.562, tangentMode=RCTM_User, time=0.3, value=1.1103}
          {interpMode=RCIM_Cubic, leaveTangent=0.3678, time=0.6, value=0}
         ]
        }}}
      CurveIndex=&Divide_Float(A=$Particles.NormalizedAge, B=0.5)
     )
    )
   ) {isSetParameters=true}
   &Drag(
    Drag=&FloatFromCurve(FloatCurve=DataInterface() {propertyValues {Curve {keys [<at 0 0>,<at 0.2138 1>,<at 1 1>]}}})
    Use Linear Drag=true
    Use Rotational Drag=true
   )
   &GravityForce(Gravity=&Lerp_Vector(A=<0 0 450>, B=<0 0 -450>, Alpha=$User.C_GravityMod))
   &GravityForce(
    Coordinate Space=::Simulation
    Gravity=&Lerp_Vector(
     A=<0 0 0>
     B=&AddVectorToPosition(
      Position=&MergeHelper_VectorSwitch(
       0=&Multiply_VectorByFloat(
        Vector=&ReturnMeshOrientationAxis(Mesh Orientation Axis=::XAxis, Mesh Orientation Quat=$Particles.SkeletalMeshLocation.SampledRotation)
        Float=-300
       )
       1=&Multiply_VectorByFloat(
        Vector=&ReturnMeshOrientationAxis(Mesh Orientation Axis=::XAxis, Mesh Orientation Quat=$Particles.SkeletalMeshLocation.SampledRotation)
        Float=300
       )
       2=?
       3=?
       Index=$Particles.PartitionParticles.Partition
       Num=::NewEnumerator0
      )
      Vector=<0 0 1>
     )
     Alpha=$User.C_GravityForce
    )
   ) {module=GravityForce001}
   &SolveForcesAndVelocity(Clamp Velocity=false, Limit Acceleration=false)
  ]
  rendererClasses [&NiagaraRibbonRendererProperties]
  renderers [
   {rendererClass=&NiagaraRibbonRendererProperties}
   {rendererClass=&NiagaraRibbonRendererProperties, rendererIndex=1}
   {rendererClass=&NiagaraRibbonRendererProperties, rendererIndex=2}
   {rendererClass=&NiagaraRibbonRendererProperties, rendererIndex=3}
   {rendererClass=&NiagaraRibbonRendererProperties, rendererIndex=4}
   {rendererClass=&NiagaraRibbonRendererProperties, rendererIndex=5}
   {rendererClass=&NiagaraRibbonRendererProperties, rendererIndex=6}
   {rendererClass=&NiagaraRibbonRendererProperties, rendererIndex=7}
  ]
 }
 NE_Sparks {
  simTarget=GPUComputeSim
  emitterUpdateScript [
   &EmitterState(
    Inactive Response=::NewEnumerator1
    Life Cycle Mode=::NewEnumerator1
    Loop Behavior=::NewEnumerator1
    Loop Duration Mode=::NewEnumerator1
    UseLoopDelay=false
   )
   &SpawnRate(SpawnGroup=2, SpawnRate=200, Use Spawn Probability=false)
  ]
  particleSpawnScript [
   /SetVariables(Particles.VisibilityTag:NiagaraBool=/Niagara/DynamicInputs/Helpers/MaskBoolBySpawnGroup(
     Masked Bool:NiagaraBool=false
     Passthrough Bool:NiagaraBool=true
     Spawn Group Mask:NiagaraInt32=&RandomRangeInt(Minimum=0, Maximum=162351520)
    )) {isSetParameters=true}
   &InitializeParticle(
    Color=<rgba 10 2.1092 0.6029 0.5>
    Lifetime=&RandomRangeFloat(Minimum=0.3, Maximum=0.5)
    Position Offset=<0 0 40>
    Sprite Size=&Multiply_Vector2D_ByFloat(
     Float=&Multiply_Float(A=0.8, B=&RandomRangeFloat(Minimum=0.3333, Maximum=1))
     Vector2D=<15 40>
    )
    Sprite Size Mode=::NewEnumerator4
    UsePositionOffset=true
    Write Material Random=true
   ) {module=InitializeParticle001}
   &PartitionParticles()
   &SkeletalMeshLocation(
    Bone / Socket Index=$Particles.PartitionParticles.Partition
    Bone Sampling Mode=::NewEnumerator4
    MaskSpawnGroup=false
    Orientation Sampling=::NewEnumerator1
    Position Sampling=::NewEnumerator1
    Skeletal Mesh=$System.SkeletalMeshUser
   )
   &ShapeLocation(
    Handle Radius=0
    Large Radius=20
    MaskSpawnGroup=false
    Offset=<0 -15 0>
    Offset Mode=::NewEnumerator0
    Override Seed=false
    Shape Primitive=::NewEnumerator3
    UseOwnerScale=true
   )
   &AddVelocity(
    Cone Angle=60
    Cone Axis=&MergeHelper_VectorSwitch(
     0=&Multiply_VectorByFloat(
      Vector=&ReturnMeshOrientationAxis(Mesh Orientation Axis=::XAxis, Mesh Orientation Quat=$Particles.SkeletalMeshLocation.SampledRotation)
      Float=-1
     )
     1=&ReturnMeshOrientationAxis(Mesh Orientation Axis=::XAxis, Mesh Orientation Quat=$Particles.SkeletalMeshLocation.SampledRotation)
     2=?
     3=?
     Index=$Particles.PartitionParticles.Partition
     Num=::NewEnumerator0
    )
    Module.UseConeFalloff=false
    Module.UseVelDistribution=false
    Override Seed=false
    Velocity Mode=::NewEnumerator2
    Velocity Speed=&RandomRangeFloat(Minimum=150, Maximum=300)
   ) {module=AddVelocity001}
  ]
  particleUpdateScript [
   &ParticleState()
   &ScaleSpriteSize(
    Scale Factor=&MakeVector2D(
     X=&FloatFromCurve(FloatCurve=DataInterface() {propertyValues {Curve {keys [
          {interpMode=RCIM_Cubic, leaveTangent=11.7691, tangentMode=RCTM_Break, time=0, value=0}
          {arriveTangent=0.0629, interpMode=RCIM_Cubic, leaveTangent=0.0629, tangentMode=RCTM_User, time=0.3, value=1}
          {interpMode=RCIM_Cubic, leaveTangent=0.027, tangentMode=RCTM_Break, time=0.35, value=1}
          {arriveTangent=0.0384, tangentMode=RCTM_User, time=1, value=0}
         ]}}})
     Y=&FloatFromCurve(FloatCurve=DataInterface() {propertyValues {Curve {keys [
          {interpMode=RCIM_Cubic, leaveTangent=11.7691, tangentMode=RCTM_Break, time=0, value=0}
          {arriveTangent=0.0629, interpMode=RCIM_Cubic, leaveTangent=0.0629, tangentMode=RCTM_User, time=0.3, value=1}
          {interpMode=RCIM_Cubic, leaveTangent=0.027, tangentMode=RCTM_Break, time=0.35, value=1}
         ]}}})
    )
    Scale Sprite Size Mode=::NewEnumerator2
   )
   &GravityForce(Gravity=&RandomRangeVector(Minimum=<0 0 200>, Maximum=<0 0 500>))
   &Drag(
    Drag=&FloatFromCurve(FloatCurve=DataInterface() {propertyValues {Curve {defaultValue=1, keys [<at 0 0.5 RCIM_Cubic>,<at 1 3 RCIM_Cubic>]}}}, Scale Curve=2)
    Use Linear Drag=true
    Use Rotational Drag=true
   )
   &SolveForcesAndVelocity(Clamp Velocity=false, Limit Acceleration=false)
   &ScaleSpriteSizeBySpeed(Max Scale Factor=<1 2>, Velocity Threshold=3500)
   &ScaleSpriteSize(
    Scale Sprite Size Mode=::NewEnumerator0
    Uniform Scale Factor=&FloatFromCurve(
     FloatCurve=DataInterface() {propertyValues {Curve {keys [<at 0.7 1>,<at 1 0>]}}}
     Scale Curve=&NormalizeDistanceRange(
      Start Position:NiagaraPosition=$Engine.Emitter.SimulationPosition
      End Position:NiagaraPosition=$Particles.Position
      Distance:NiagaraFloat=300
      Invert Normalized Range:NiagaraBool=false
     )
    )
   ) {module=ScaleSpriteSize001}
   &FN_Sprite_Attributes(
    Color=&MultiplyLinearColor(Linear Color=$Particles.Color, Scale Factor=<0.025 0.025 0.025 0.5>)
    Create Color Attribute=true
    Create SpriteSize Attribute=true
    SpriteSize=&Multiply_Vector2D(A=$Particles.SpriteSize, B=<5 2>)
    Write to Module Namespace=true
   )
   &FN_Sprite_Attributes(
    Color=<rgba 0.08 0.0776 0.0791 1>
    Create Color Attribute=true
    Create Position Attribute=true
    Create Sprite Rotation Attribute=true
    Create SpriteSize Attribute=true
    Position:NiagaraPosition=&Multiply_Vector(
     A=Linked() {linkedVariable {name=Particles.Position, type=NiagaraPosition}}
     B=&RandomRangeVector(Minimum=<1.1 1.1 1>, Maximum=<2 2 1>)
    )
    SpriteRotation:NiagaraFloat=&RandomRangeFloat(Minimum=0, Maximum=360)
    SpriteSize=<0.001 0.001>
    Write to Module Namespace=true
   ) {module=FN_Sprite_Attributes001}
   &ScaleColor(
    Scale Alpha=&FloatFromCurve(FloatCurve=DataInterface() {propertyValues {Curve {
        defaultValue=1
        keys [
         {arriveTangent=9.0725, interpMode=RCIM_Cubic, leaveTangent=9.0725, tangentMode=RCTM_User, time=0, value=0}
         {arriveTangent=-2.5249, interpMode=RCIM_Cubic, leaveTangent=-2.5249, tangentMode=RCTM_User, time=0.1408, value=0.9847}
         {arriveTangent=-0.0405, interpMode=RCIM_Cubic, leaveTangent=-0.0405, tangentMode=RCTM_User, time=1, value=0}
        ]
       }}})
    ScaleA=true
    ScaleRGB=true
   )
   &CameraOffset(Camera Offset Amount=15)
  ]
  rendererClasses [&NiagaraSpriteRendererProperties]
  renderers [
   {rendererClass=&NiagaraSpriteRendererProperties}
   {rendererClass=&NiagaraSpriteRendererProperties, rendererIndex=1}
  ]
 }
 SplashSprites {
  emitterUpdateScript [
   &EmitterState()
   &SpawnParticlesFromOtherEmitter(
    Attribute Reader=DataInterface() {propertyValues {EmitterBinding {bindingMode=Other, emitterName=Collision_Position}}}
    Calculate Spawn Rate Per Particle=false
    Spawn Rate=15
    Use Maximum Age=false
    Use Minimum Age=false
   )
  ]
  particleSpawnScript [
   &InitializeParticle(
    Color=&RandomRangeLinearColor(Minimum=<rgba 0.1042 0.7014 1 1>, Maximum=<rgba 0 1.5 1.25 1>)
    Lifetime Max=0.7
    Lifetime Min=0.5
    Lifetime Mode=::NewEnumerator1
    Position Mode=::NewEnumerator0
    Sprite Rotation Mode=::NewEnumerator2
    Sprite Size Mode=::NewEnumerator1
    Sprite UV Mode=::NewEnumerator3
    Uniform Sprite Size Max=120
    Uniform Sprite Size Min=80
    Write Material Random=true
   )
   &SampleParticlesFromOtherEmitter()
   /SetVariables(Particles.SubImageIndex=&RandomRangeFloat(Minimum=0, Maximum=3)) {isSetParameters=true, module=SetVariables_#3}
   &ShapeLocation(MaskSpawnGroup=false, Override Seed=false, Sphere Radius=2, Sphere Surface Distribution=0.2, UseOwnerScale=true)
   &AddVelocity(
    Module.Use Falloff Distance=false
    Origin Offset=<-5 0 -2>
    Override Seed=false
    Velocity Mode=::NewEnumerator1
    Velocity Speed=&Multiply_Float(
     A=&Multiply_Float(
      A=&RandomRangeFloat(Minimum=100, Maximum=200)
      B=&FN_PickFloatDependOnPlatform(TargetPlatform=DataInterface() {propertyValues {Platforms {qualityLevelMask=-2}}}, TargetPlatformActive=1, TargetPlatformNoActive=0.7)
     )
     B=$System.FrontendScale
    )
   )
  ]
  particleUpdateScript [
   &ParticleState()
   &SpriteRotationRate(Rotation Rate=&FloatFromCurve(
     FloatCurve=DataInterface() {propertyValues {Curve {
        defaultValue=1
        keys [{interpMode=RCIM_Cubic, tangentMode=RCTM_User, time=0, value=1}, {interpMode=RCIM_Cubic, tangentMode=RCTM_User, time=1, value=0}]
       }}}
     Scale Curve=&Multiply_FloatByInt(
      Float=&RandomRangeFloat(Minimum=60, Maximum=100)
      Integer=&SelectIntFromArray(
       Array Sampling Mode=::NewEnumerator1
       Direct Array Index=?
       Int Selection Array=DataInterface() {propertyValues {IntData [-1,1]}}
       Random Seed=0
      )
     )
    ))
   &GravityForce(Gravity=&Multiply_VectorByFloat(
     Vector=&Multiply_VectorByFloat(Vector=<0 0 -600>, Float=&RandomRangeFloat(Minimum=0.7, Maximum=1))
     Float=$System.FrontendScale
    ))
   &Drag(Drag=&RandomRangeFloat(Minimum=2, Maximum=3), Use Linear Drag=true, Use Rotational Drag=true)
   &SolveForcesAndVelocity(Clamp Velocity=false, Limit Acceleration=false)
   &ScaleSpriteSize(Uniform Curve Sprite Scale=DataInterface() {propertyValues {Curve {keys [
        {arriveTangent=1.7973, interpMode=RCIM_Cubic, leaveTangent=1.7973, tangentMode=RCTM_User, time=0, value=0.1}
        {interpMode=RCIM_Cubic, tangentMode=RCTM_User, time=1, value=1}
       ]}}})
   &ScaleColor(
    Scale Alpha=&FloatFromCurve(FloatCurve=DataInterface() {propertyValues {Curve {
        defaultValue=1
        keys [{interpMode=RCIM_Cubic, tangentMode=RCTM_User, time=0, value=1}, {interpMode=RCIM_Cubic, tangentMode=RCTM_User, time=1, value=0}]
       }}})
    Scale RGB=&MakeVectorFromLinearColorRGB(Color=&ColorFromCurve(ColorCurve=DataInterface() {propertyValues {
        AlphaCurve {keys [<at 0 1 RCIM_Cubic>]}
        BlueCurve {keys [<at 0 1>,<at 0.8321 0.1>]}
        GreenCurve {keys [<at 0 1>,<at 0.8321 0.0825>]}
        RedCurve {keys [<at 0 1>,<at 0.8321 0.08>]}
       }}))
    ScaleA=true
    ScaleRGB=true
   )
   &ScaleColor(
    Scale RGB=&FN_FrontEndVector3Modifier(
     Gameplay=&MakeVectorFromLinearColorRGB(Color=&FNTimeOfDayColor(Night Color=<rgba 0.32 0.34 0.4 1>, Morning Color=<rgba 0.6 0.468 0.42 1>, Day Color=<rgba 1 1 1 1>, Evening Color=<rgba 0.8 0.52 0.4 1>))
     Front End=&MakeVectorFromLinearColorRGB(Color=<rgba 0 1.3333 2 1>)
    )
    ScaleA=true
    ScaleRGB=true
   ) {enabled=false, module=ScaleColor001}
   &DynamicMaterialParameters(
    Index 0 Param 1=&RandomRangeFloat(Minimum=0, Maximum=1)
    Index 0 Param 2=&FloatFromCurve(
     FloatCurve=DataInterface() {propertyValues {Curve {
        defaultValue=1
        keys [
         {arriveTangent=-2.3912, interpMode=RCIM_Cubic, leaveTangent=-2.3912, tangentMode=RCTM_User, time=0, value=1}
         {interpMode=RCIM_Cubic, tangentMode=RCTM_User, time=0.7, value=0}
        ]
       }}}
     Scale Curve=&RandomRangeFloat(Minimum=0.5, Maximum=1)
    )
    Index 0 Param 3=&Lerp_Float(
     A=0
     B=3
     Alpha=&FloatFromCurve(FloatCurve=DataInterface() {propertyValues {Curve {
         defaultValue=0
         keys [
          {interpMode=RCIM_Cubic, tangentMode=RCTM_User, time=0, value=0}
          {arriveTangent=3.1748, interpMode=RCIM_Cubic, leaveTangent=3.1748, tangentMode=RCTM_User, time=1, value=1}
         ]
        }}})
    )
    Index 0 Param 4=&FloatFromCurve(FloatCurve=DataInterface() {propertyValues {Curve {
        defaultValue=0
        keys [
         {arriveTangent=2.4825, interpMode=RCIM_Cubic, leaveTangent=2.4825, tangentMode=RCTM_User, time=0, value=0}
         {interpMode=RCIM_Cubic, tangentMode=RCTM_User, time=1, value=1}
        ]
       }}})
    Module.Index0Param0Write=true
    Module.Index0Param1Write=true
    Module.Index0Param2Write=true
    Module.Index0Param3Write=true
   )
   &ScaleColor(Scale Alpha=0.2, ScaleA=true, ScaleRGB=true) {module=ScaleColor002}
  ]
  rendererClasses [&NiagaraSpriteRendererProperties]
  renderers [{rendererClass=&NiagaraSpriteRendererProperties}]
 }
 SplashSprites001 {
  emitterUpdateScript [
   &EmitterState(Inactive Response=::NewEnumerator1, Life Cycle Mode=::NewEnumerator1, Loop Duration=1, UseLoopDelay=false)
   &SpawnParticlesFromOtherEmitter(
    Attribute Reader=DataInterface() {propertyValues {EmitterBinding {bindingMode=Other, emitterName=Toon_Water}}}
    Spawn Rate=20
    Spawn Rate Per Particle Cap=10
    Use Maximum Age=false
    Use Minimum Age=false
   )
  ]
  particleSpawnScript [
   &InitializeParticle(
    Color=&Multiply_LinearColorByFloat(
     Linear Color=&RandomRangeLinearColor(Minimum=<rgba 0.1042 0.7014 1 1>, Maximum=<rgba 0 1.5 1.25 1>)
     Float=1
    )
    Lifetime Max=1.2
    Lifetime Min=0.8
    Lifetime Mode=::NewEnumerator1
    Position Mode=::NewEnumerator0
    Sprite Rotation Mode=::NewEnumerator2
    Sprite Size Mode=::NewEnumerator1
    Sprite UV Mode=::NewEnumerator3
    Uniform Sprite Size Max=120
    Uniform Sprite Size Min=80
    Write Material Random=true
   )
   &SampleParticlesFromOtherEmitter(
    Applied Velocity Scale=0.2
    Maximum Sampled Particle Age=0.15
    Module.MaxSampledAgeCutoff=true
    Module.MinSampledAgeCutoff=false
    Reject Particle By Sampled Age=true
    Velocity Sampling=::NewEnumerator0
   )
   /SetVariables(Particles.SubImageIndex=&RandomRangeFloat(Minimum=0, Maximum=3)) {isSetParameters=true, module=SetVariables_#3}
   &ShapeLocation(MaskSpawnGroup=false, Override Seed=false, Sphere Radius=2, Sphere Surface Distribution=0.2, UseOwnerScale=true) {enabled=false}
   &AddVelocity(
    Module.Use Falloff Distance=false
    Origin Offset=<-5 0 -2>
    Override Seed=false
    Velocity Mode=::NewEnumerator1
    Velocity Speed=&Multiply_Float(
     A=&Multiply_Float(
      A=&RandomRangeFloat(Minimum=100, Maximum=200)
      B=&FN_PickFloatDependOnPlatform(TargetPlatform=DataInterface() {propertyValues {Platforms {qualityLevelMask=-2}}}, TargetPlatformActive=1, TargetPlatformNoActive=0.7)
     )
     B=$System.FrontendScale
    )
   ) {enabled=false}
   &AddVelocity(Velocity=<0 0 250>) {enabled=false, module=AddVelocity001}
   &ApplyInitialForces()
  ]
  particleUpdateScript [
   &ParticleState()
   &SpriteRotationRate(Rotation Rate=&FloatFromCurve(
     FloatCurve=DataInterface() {propertyValues {Curve {
        defaultValue=1
        keys [{interpMode=RCIM_Cubic, tangentMode=RCTM_User, time=0, value=1}, {interpMode=RCIM_Cubic, tangentMode=RCTM_User, time=1, value=0}]
       }}}
     Scale Curve=&Multiply_FloatByInt(
      Float=&RandomRangeFloat(Minimum=60, Maximum=100)
      Integer=&SelectIntFromArray(
       Array Sampling Mode=::NewEnumerator1
       Direct Array Index=?
       Int Selection Array=DataInterface() {propertyValues {IntData [-1,1]}}
       Random Seed=0
      )
     )
    ))
   &GravityForce(Gravity=&Multiply_VectorByFloat(
     Vector=&Multiply_VectorByFloat(Vector=<0 0 -600>, Float=&RandomRangeFloat(Minimum=0.7, Maximum=1))
     Float=1
    ))
   &Drag(Drag=&RandomRangeFloat(Minimum=2, Maximum=3), Use Linear Drag=true, Use Rotational Drag=true)
   &SolveForcesAndVelocity(Clamp Velocity=false, Limit Acceleration=false)
   &ScaleSpriteSize(Uniform Curve Sprite Scale=DataInterface() {propertyValues {Curve {keys [
        {arriveTangent=1.7973, interpMode=RCIM_Cubic, leaveTangent=1.7973, tangentMode=RCTM_User, time=0, value=0.1}
        {interpMode=RCIM_Cubic, tangentMode=RCTM_User, time=1, value=1}
       ]}}})
   &ScaleColor(
    Scale Alpha=&FloatFromCurve(FloatCurve=DataInterface() {propertyValues {Curve {
        defaultValue=1
        keys [{interpMode=RCIM_Cubic, tangentMode=RCTM_User, time=0, value=1}, {interpMode=RCIM_Cubic, tangentMode=RCTM_User, time=1, value=0}]
       }}})
    Scale RGB=&MakeVectorFromLinearColorRGB(Color=&ColorFromCurve(ColorCurve=DataInterface() {propertyValues {
        AlphaCurve {keys [<at 0 1 RCIM_Cubic>]}
        BlueCurve {keys [<at 0 1>,<at 0.8321 0.1>]}
        GreenCurve {keys [<at 0 1>,<at 0.8321 0.0825>]}
        RedCurve {keys [<at 0 1>,<at 0.8321 0.08>]}
       }}))
    ScaleA=true
    ScaleRGB=true
   )
   &ScaleColor(
    Scale RGB=&FN_FrontEndVector3Modifier(
     Gameplay=&MakeVectorFromLinearColorRGB(Color=&FNTimeOfDayColor(Night Color=<rgba 0.32 0.34 0.4 1>, Morning Color=<rgba 0.6 0.468 0.42 1>, Day Color=<rgba 1 1 1 1>, Evening Color=<rgba 0.8 0.52 0.4 1>))
     Front End=&MakeVectorFromLinearColorRGB(Color=<rgba 0 1.3333 2 1>)
    )
    ScaleA=true
    ScaleRGB=true
   ) {enabled=false, module=ScaleColor001}
   &DynamicMaterialParameters(
    Index 0 Param 1=&RandomRangeFloat(Minimum=0, Maximum=1)
    Index 0 Param 2=&FloatFromCurve(
     FloatCurve=DataInterface() {propertyValues {Curve {
        defaultValue=1
        keys [
         {arriveTangent=-2.3912, interpMode=RCIM_Cubic, leaveTangent=-2.3912, tangentMode=RCTM_User, time=0, value=1}
         {interpMode=RCIM_Cubic, tangentMode=RCTM_User, time=0.7, value=0}
        ]
       }}}
     Scale Curve=&RandomRangeFloat(Minimum=0.5, Maximum=1)
    )
    Index 0 Param 3=&Lerp_Float(
     A=0
     B=3
     Alpha=&FloatFromCurve(FloatCurve=DataInterface() {propertyValues {Curve {
         defaultValue=0
         keys [
          {interpMode=RCIM_Cubic, tangentMode=RCTM_User, time=0, value=0}
          {arriveTangent=3.1748, interpMode=RCIM_Cubic, leaveTangent=3.1748, tangentMode=RCTM_User, time=1, value=1}
         ]
        }}})
    )
    Index 0 Param 4=&FloatFromCurve(FloatCurve=DataInterface() {propertyValues {Curve {
        defaultValue=0
        keys [
         {arriveTangent=2.4825, interpMode=RCIM_Cubic, leaveTangent=2.4825, tangentMode=RCTM_User, time=0, value=0}
         {interpMode=RCIM_Cubic, tangentMode=RCTM_User, time=1, value=1}
        ]
       }}})
    Module.Index0Param0Write=true
    Module.Index0Param1Write=true
    Module.Index0Param2Write=true
    Module.Index0Param3Write=true
   )
   &ScaleColor(Scale Alpha=0.3, ScaleA=true, ScaleRGB=true) {module=ScaleColor002}
  ]
  rendererClasses [&NiagaraSpriteRendererProperties]
  renderers [{rendererClass=&NiagaraSpriteRendererProperties}]
 }
 SplashSprites002 {
  emitterUpdateScript [
   &EmitterState(Inactive Response=::NewEnumerator1, Life Cycle Mode=::NewEnumerator1, Loop Duration=1, UseLoopDelay=false)
   &SpawnParticlesFromOtherEmitter(
    Attribute Reader=DataInterface() {propertyValues {EmitterBinding {bindingMode=Other, emitterName=Toon_Water}}}
    Spawn Rate=60
    Spawn Rate Per Particle Cap=30
    Use Maximum Age=false
    Use Minimum Age=false
   )
  ]
  particleSpawnScript [
   &InitializeParticle(
    Color=&Multiply_LinearColorByFloat(
     Linear Color=&RandomRangeLinearColor(Minimum=<rgba 0.1042 0.7014 1 1>, Maximum=<rgba 0 1.5 1.25 1>)
     Float=10
    )
    Lifetime Max=1.2
    Lifetime Min=0.8
    Lifetime Mode=::NewEnumerator1
    Position Mode=::NewEnumerator0
    Sprite Rotation Mode=::NewEnumerator2
    Sprite Size Mode=::NewEnumerator1
    Sprite UV Mode=::NewEnumerator3
    Uniform Sprite Size Max=120
    Uniform Sprite Size Min=80
    Write Material Random=true
   )
   &SampleParticlesFromOtherEmitter(
    Maximum Sampled Particle Age=0.05
    Minimum Sampled Particle Age=0.032
    Module.MaxSampledAgeCutoff=true
    Module.MinSampledAgeCutoff=true
    Reject Particle By Sampled Age=true
    Velocity Sampling=::NewEnumerator0
   )
   /SetVariables(Particles.SubImageIndex=&RandomRangeFloat(Minimum=0, Maximum=3)) {isSetParameters=true, module=SetVariables_#3}
   &ShapeLocation(MaskSpawnGroup=false, Override Seed=false, Sphere Radius=2, Sphere Surface Distribution=0.2, UseOwnerScale=true) {enabled=false}
   &AddVelocity(
    Module.Use Falloff Distance=false
    Origin Offset=<-5 0 -2>
    Override Seed=false
    Velocity Mode=::NewEnumerator1
    Velocity Speed=&Multiply_Float(
     A=&Multiply_Float(
      A=&RandomRangeFloat(Minimum=100, Maximum=200)
      B=&FN_PickFloatDependOnPlatform(TargetPlatform=DataInterface() {propertyValues {Platforms {qualityLevelMask=-2}}}, TargetPlatformActive=1, TargetPlatformNoActive=0.7)
     )
     B=$System.FrontendScale
    )
   ) {enabled=false}
   &CameraOffset(Camera Offset Amount=-40)
   &AddVelocity(Velocity=<0 0 250>) {module=AddVelocity001}
   &ApplyInitialForces()
  ]
  particleUpdateScript [
   &ParticleState()
   &SpriteRotationRate(Rotation Rate=&FloatFromCurve(
     FloatCurve=DataInterface() {propertyValues {Curve {
        defaultValue=1
        keys [{interpMode=RCIM_Cubic, tangentMode=RCTM_User, time=0, value=1}, {interpMode=RCIM_Cubic, tangentMode=RCTM_User, time=1, value=0}]
       }}}
     Scale Curve=&Multiply_FloatByInt(
      Float=&RandomRangeFloat(Minimum=60, Maximum=100)
      Integer=&SelectIntFromArray(
       Array Sampling Mode=::NewEnumerator1
       Direct Array Index=?
       Int Selection Array=DataInterface() {propertyValues {IntData [-1,1]}}
       Random Seed=0
      )
     )
    ))
   &GravityForce(Gravity=&Multiply_VectorByFloat(
     Vector=&Multiply_VectorByFloat(Vector=<0 0 -980>, Float=&RandomRangeFloat(Minimum=0.7, Maximum=1))
     Float=2
    ))
   &Drag(Drag=&RandomRangeFloat(Minimum=2, Maximum=3), Use Linear Drag=true, Use Rotational Drag=true)
   &SolveForcesAndVelocity(Clamp Velocity=false, Limit Acceleration=false)
   &ScaleSpriteSize(Uniform Curve Sprite Scale=DataInterface() {propertyValues {Curve {keys [
        {arriveTangent=1.7973, interpMode=RCIM_Cubic, leaveTangent=1.7973, tangentMode=RCTM_User, time=0, value=0.1}
        {interpMode=RCIM_Cubic, tangentMode=RCTM_User, time=1, value=1}
       ]}}})
   &ScaleColor(
    Scale Alpha=&FloatFromCurve(FloatCurve=DataInterface() {propertyValues {Curve {
        defaultValue=1
        keys [{interpMode=RCIM_Cubic, tangentMode=RCTM_User, time=0, value=1}, {interpMode=RCIM_Cubic, tangentMode=RCTM_User, time=1, value=0}]
       }}})
    Scale RGB=&MakeVectorFromLinearColorRGB(Color=&ColorFromCurve(ColorCurve=DataInterface() {propertyValues {
        AlphaCurve {keys [<at 0 1 RCIM_Cubic>]}
        BlueCurve {keys [<at 0 1>,<at 0.8321 0.1>]}
        GreenCurve {keys [<at 0 1>,<at 0.8321 0.0825>]}
        RedCurve {keys [<at 0 1>,<at 0.8321 0.08>]}
       }}))
    ScaleA=true
    ScaleRGB=true
   )
   &ScaleColor(
    Scale RGB=&FN_FrontEndVector3Modifier(
     Gameplay=&MakeVectorFromLinearColorRGB(Color=&FNTimeOfDayColor(Night Color=<rgba 0.32 0.34 0.4 1>, Morning Color=<rgba 0.6 0.468 0.42 1>, Day Color=<rgba 1 1 1 1>, Evening Color=<rgba 0.8 0.52 0.4 1>))
     Front End=&MakeVectorFromLinearColorRGB(Color=<rgba 0 1.3333 2 1>)
    )
    ScaleA=true
    ScaleRGB=true
   ) {enabled=false, module=ScaleColor001}
   &DynamicMaterialParameters(
    Index 0 Param 1=&RandomRangeFloat(Minimum=0, Maximum=1)
    Index 0 Param 2=&FloatFromCurve(
     FloatCurve=DataInterface() {propertyValues {Curve {
        defaultValue=1
        keys [
         {arriveTangent=-2.3912, interpMode=RCIM_Cubic, leaveTangent=-2.3912, tangentMode=RCTM_User, time=0, value=1}
         {interpMode=RCIM_Cubic, tangentMode=RCTM_User, time=0.7, value=0}
        ]
       }}}
     Scale Curve=&RandomRangeFloat(Minimum=0.5, Maximum=1)
    )
    Index 0 Param 3=&Lerp_Float(
     A=0
     B=3
     Alpha=&FloatFromCurve(FloatCurve=DataInterface() {propertyValues {Curve {
         defaultValue=0
         keys [
          {interpMode=RCIM_Cubic, tangentMode=RCTM_User, time=0, value=0}
          {arriveTangent=3.1748, interpMode=RCIM_Cubic, leaveTangent=3.1748, tangentMode=RCTM_User, time=1, value=1}
         ]
        }}})
    )
    Index 0 Param 4=&FloatFromCurve(FloatCurve=DataInterface() {propertyValues {Curve {
        defaultValue=0
        keys [
         {arriveTangent=2.4825, interpMode=RCIM_Cubic, leaveTangent=2.4825, tangentMode=RCTM_User, time=0, value=0}
         {interpMode=RCIM_Cubic, tangentMode=RCTM_User, time=1, value=1}
        ]
       }}})
    Module.Index0Param0Write=true
    Module.Index0Param1Write=true
    Module.Index0Param2Write=true
    Module.Index0Param3Write=true
   )
   &ScaleColor(Scale Alpha=0.3, ScaleA=true, ScaleRGB=true) {module=ScaleColor002}
  ]
  rendererClasses [&NiagaraSpriteRendererProperties]
  renderers [{rendererClass=&NiagaraSpriteRendererProperties}]
 }
 SplashSprites003 {
  emitterUpdateScript [
   &EmitterState()
   &SpawnParticlesFromOtherEmitter(
    Attribute Reader=DataInterface() {propertyValues {EmitterBinding {bindingMode=Other, emitterName=Collision_Position}}}
    Calculate Spawn Rate Per Particle=false
    Spawn Rate=500
    Use Maximum Age=false
    Use Minimum Age=false
   )
  ]
  particleSpawnScript [
   &InitializeParticle(
    Color=&RandomRangeLinearColor(Minimum=<rgba 0.4271 0.5229 1 1>, Maximum=<rgba 0.5549 0.9336 1.5 1>)
    Lifetime Max=0.7
    Lifetime Min=0.5
    Lifetime Mode=::NewEnumerator1
    Position Mode=::NewEnumerator0
    Sprite Rotation Mode=::NewEnumerator2
    Sprite Size Mode=::NewEnumerator1
    Sprite UV Mode=::NewEnumerator3
    Uniform Sprite Size Max=200
    Uniform Sprite Size Min=100
    Write Material Random=true
   )
   &SampleParticlesFromOtherEmitter()
   /SetVariables(Particles.SubImageIndex=&RandomRangeFloat(Minimum=0, Maximum=3)) {enabled=false, isSetParameters=true, module=SetVariables_#3}
   &ShapeLocation(MaskSpawnGroup=false, Override Seed=false, Sphere Radius=2, Sphere Surface Distribution=0.2, UseOwnerScale=true)
   &AddVelocity(
    Module.Use Falloff Distance=false
    Origin Offset=<-5 0 -2>
    Override Seed=false
    Velocity Mode=::NewEnumerator1
    Velocity Speed=&Multiply_Float(
     A=&Multiply_Float(
      A=&RandomRangeFloat(Minimum=100, Maximum=200)
      B=&FN_PickFloatDependOnPlatform(TargetPlatform=DataInterface() {propertyValues {Platforms {qualityLevelMask=-2}}}, TargetPlatformActive=1, TargetPlatformNoActive=0.7)
     )
     B=0.001
    )
   )
   &SpriteFacingAndAlignment(Sprite Facing=<0 0 0>, WriteAlignment=true, WriteFacing=true) {enabled=false}
  ]
  particleUpdateScript [
   &Collision(
    Age Colliding Particles=false
    KillOnCollision=false
    Module.Control Rest State Via Surface Normal=false
    Module.EnableMaxCollisionCount=false
    Module.Randomize Collision Normal Vector=false
   ) {enabled=false}
   &ParticleState()
   &SpriteRotationRate(Rotation Rate=&FloatFromCurve(
     FloatCurve=DataInterface() {propertyValues {Curve {
        defaultValue=1
        keys [{interpMode=RCIM_Cubic, tangentMode=RCTM_User, time=0, value=1}, {interpMode=RCIM_Cubic, tangentMode=RCTM_User, time=1, value=0}]
       }}}
     Scale Curve=&Multiply_FloatByInt(
      Float=&RandomRangeFloat(Minimum=60, Maximum=100)
      Integer=&SelectIntFromArray(
       Array Sampling Mode=::NewEnumerator1
       Direct Array Index=?
       Int Selection Array=DataInterface() {propertyValues {IntData [-1,1]}}
       Random Seed=0
      )
     )
    ))
   &GravityForce(Gravity=&Multiply_VectorByFloat(
     Vector=&Multiply_VectorByFloat(Vector=<0 0 -600>, Float=&RandomRangeFloat(Minimum=0.7, Maximum=1))
     Float=$System.FrontendScale
    ))
   &Drag(Drag=&RandomRangeFloat(Minimum=2, Maximum=3), Use Linear Drag=true, Use Rotational Drag=true)
   &SolveForcesAndVelocity(Clamp Velocity=false, Limit Acceleration=false)
   &ScaleSpriteSize(Uniform Curve Sprite Scale=DataInterface() {propertyValues {Curve {keys [
        {arriveTangent=1.7973, interpMode=RCIM_Cubic, leaveTangent=1.7973, tangentMode=RCTM_User, time=0, value=0.1}
        {interpMode=RCIM_Cubic, tangentMode=RCTM_User, time=1, value=1}
       ]}}})
   &ScaleColor(
    Scale Alpha=&FloatFromCurve(FloatCurve=DataInterface() {propertyValues {Curve {
        defaultValue=1
        keys [{interpMode=RCIM_Cubic, tangentMode=RCTM_User, time=0, value=1}, {interpMode=RCIM_Cubic, tangentMode=RCTM_User, time=1, value=0}]
       }}})
    Scale RGB=&MakeVectorFromLinearColorRGB(Color=&ColorFromCurve(ColorCurve=DataInterface() {propertyValues {
        AlphaCurve {keys [<at 0 1 RCIM_Cubic>,<at 0.7284 0>]}
        BlueCurve {keys [<at 0 1>,<at 0.8321 0.625>]}
        GreenCurve {keys [<at 0 1>,<at 0.8321 0.5154>]}
        RedCurve {keys [<at 0 1>,<at 0.8321 0.5>]}
       }}))
    ScaleA=true
    ScaleRGB=true
   )
   &ScaleColor(
    Scale RGB=&FN_FrontEndVector3Modifier(
     Gameplay=&MakeVectorFromLinearColorRGB(Color=&FNTimeOfDayColor(Night Color=<rgba 0.32 0.34 0.4 1>, Morning Color=<rgba 0.6 0.468 0.42 1>, Day Color=<rgba 1 1 1 1>, Evening Color=<rgba 0.8 0.52 0.4 1>))
     Front End=&MakeVectorFromLinearColorRGB(Color=<rgba 0 1.3333 2 1>)
    )
    ScaleA=true
    ScaleRGB=true
   ) {enabled=false, module=ScaleColor001}
   &DynamicMaterialParameters(
    Index 0 Param 1=&RandomRangeFloat(Minimum=0, Maximum=1)
    Index 0 Param 2=&FloatFromCurve(
     FloatCurve=DataInterface() {propertyValues {Curve {
        defaultValue=1
        keys [
         {arriveTangent=-2.3912, interpMode=RCIM_Cubic, leaveTangent=-2.3912, tangentMode=RCTM_User, time=0, value=1}
         {interpMode=RCIM_Cubic, tangentMode=RCTM_User, time=0.7, value=0}
        ]
       }}}
     Scale Curve=&RandomRangeFloat(Minimum=0.5, Maximum=1)
    )
    Index 0 Param 3=&Lerp_Float(
     A=0
     B=3
     Alpha=&FloatFromCurve(FloatCurve=DataInterface() {propertyValues {Curve {
         defaultValue=0
         keys [
          {interpMode=RCIM_Cubic, tangentMode=RCTM_User, time=0, value=0}
          {arriveTangent=3.1748, interpMode=RCIM_Cubic, leaveTangent=3.1748, tangentMode=RCTM_User, time=1, value=1}
         ]
        }}})
    )
    Index 0 Param 4=&FloatFromCurve(FloatCurve=DataInterface() {propertyValues {Curve {
        defaultValue=0
        keys [
         {arriveTangent=2.4825, interpMode=RCIM_Cubic, leaveTangent=2.4825, tangentMode=RCTM_User, time=0, value=0}
         {interpMode=RCIM_Cubic, tangentMode=RCTM_User, time=1, value=1}
        ]
       }}})
    Module.Index0Param0Write=true
    Module.Index0Param1Write=true
    Module.Index0Param2Write=true
    Module.Index0Param3Write=true
   ) {enabled=false}
   /Niagara/Modules/Collision/AlignParticlesWithCollisionPlane(
    Allow Alignment on X=false
    Allow Alignment on Y=false
    Force Alignment=true
    Sprite Relative Alignment Vector:Vector3f=<1 0 0>
    Sprite Relative Facing Vector:Vector3f=<0 1 0>
   ) {enabled=false}
   &ScaleColor(Scale Alpha=0.8, ScaleA=true, ScaleRGB=true) {module=ScaleColor002}
   &ScaleColor(Scale Alpha=&FN_FrontEndFloatModifier(Gameplay=1, Front End=0.3), ScaleA=true, ScaleRGB=true) {module=ScaleColor003}
  ]
  rendererClasses [&NiagaraSpriteRendererProperties]
  renderers [{rendererClass=&NiagaraSpriteRendererProperties}]
 }
 Toon_Fire {
  emitterUpdateScript [
   &EmitterState()
   &SpawnRate(SpawnRate=100, Use Spawn Probability=false)
  ]
  particleSpawnScript [
   &InitializeParticle(Lifetime=0.22, Ribbon Width=80, Ribbon Width Mode=::NewEnumerator1, UsePositionOffset=false, Write Material Random=true)
   &PartitionParticles()
   &SkeletalMeshLocation(
    Bone / Socket Index=$Particles.PartitionParticles.Partition
    Bone Sampling Mode=::NewEnumerator4
    MaskSpawnGroup=false
    Orientation Sampling=::NewEnumerator1
    Position Sampling=::NewEnumerator1
    Skeletal Mesh=$System.SkeletalMeshUser
   )
   /SetVariables(Particles.Position=$Output.SkeletalMeshLocation.SampledPositionWithoutOffset) {isSetParameters=true, module=SetVariables_#2}
   &AddVelocity(
    Rotation Coordinate Space=::Simulation
    Velocity=&ReturnMeshOrientationAxis(Mesh Orientation Axis=::XAxis, Mesh Orientation Quat=$Particles.SkeletalMeshLocation.SampledRotation)
    Velocity Speed Scale=&MergeHelper_FloatSwitch(
     0=&Subtract_Float(A=0, B=$User.User_Velocity)
     1=$User.User_Velocity
     2=?
     3=?
     Index=$Particles.PartitionParticles.Partition
     Num=::NewEnumerator0
    )
   )
   &ApplyInitialForces()
  ]
  particleUpdateScript [
   &ParticleState()
   &ScaleColor(
    Scale Alpha=&FloatFromCurve(FloatCurve=DataInterface() {propertyValues {Curve {defaultValue=1, keys [{interpMode=RCIM_Cubic, tangentMode=RCTM_User, time=0, value=1}]}}})
    Scale RGB=&MakeVectorFromLinearColorRGB(Color=<rgba 1 1 1 1>)
    ScaleA=true
    ScaleRGB=true
   )
   /SetVariables(Particles.GlowColor=&MultiplyLinearColor(
     Linear Color=<rgba 0.3391 0.9196 1.5 1>
     Scale Factor=&MakeVector4FromVectorAndFloat(
      "Vector (RGB)"=&MakeVectorFromLinearColorRGB(Color=$System.TODColor)
      "Float (Alpha)"=&FloatFromCurve(
       FloatCurve=DataInterface() {propertyValues {Curve {
          defaultValue=1
          keys [
           {interpMode=RCIM_Cubic, tangentMode=RCTM_User, time=0, value=1}
           {arriveTangent=-1.0589, interpMode=RCIM_Cubic, leaveTangent=-1.0589, tangentMode=RCTM_User, time=1, value=0}
          ]
         }}}
       Scale Curve=$System.SwingGeneralFade
      )
     )
    )) {isSetParameters=true, module=SetVariables_#4}
   &ScaleRibbonWidth(Ribbon Width Scale=&Multiply_Float(
     A=&FloatFromCurve(FloatCurve=DataInterface() {propertyValues {Curve {
         defaultValue=0
         keys [
          {interpMode=RCIM_Cubic, tangentMode=RCTM_User, time=0, value=1}
          {arriveTangent=-1.5474, interpMode=RCIM_Cubic, leaveTangent=-1.5474, tangentMode=RCTM_User, time=1, value=0}
         ]
        }}})
     B=1
    ))
   &GravityForce(Gravity=&Lerp_Vector(A=<0 0 450>, B=<0 0 -450>, Alpha=$User.C_GravityMod))
   &GravityForce(
    Coordinate Space=::Simulation
    Gravity=&Lerp_Vector(
     A=<0 0 0>
     B=&AddVectorToPosition(
      Position=&MergeHelper_VectorSwitch(
       0=&Multiply_VectorByFloat(
        Vector=&ReturnMeshOrientationAxis(Mesh Orientation Axis=::XAxis, Mesh Orientation Quat=$Particles.SkeletalMeshLocation.SampledRotation)
        Float=-300
       )
       1=&Multiply_VectorByFloat(
        Vector=&ReturnMeshOrientationAxis(Mesh Orientation Axis=::XAxis, Mesh Orientation Quat=$Particles.SkeletalMeshLocation.SampledRotation)
        Float=300
       )
       2=?
       3=?
       Index=$Particles.PartitionParticles.Partition
       Num=::NewEnumerator0
      )
      Vector=<0 0 1>
     )
     Alpha=$User.C_GravityForce
    )
   ) {module=GravityForce001}
   &SolveForcesAndVelocity(Clamp Velocity=false, Limit Acceleration=false)
   &ScaleColor(Scale Alpha=0.7, ScaleA=true, ScaleRGB=true) {module=ScaleColor001}
  ]
  rendererClasses [&NiagaraRibbonRendererProperties]
  renderers [
   {rendererClass=&NiagaraRibbonRendererProperties}
   {rendererClass=&NiagaraRibbonRendererProperties, rendererIndex=1}
  ]
 }
 Toon_Fire002 {
  bEnabled=false
  emitterUpdateScript [
   &EmitterState()
   &SpawnRate(SpawnRate=200, Use Spawn Probability=false)
  ]
  particleSpawnScript [
   &InitializeParticle(
    Color=<rgba 1 1 1 0.001>
    Lifetime=0.5
    Ribbon Width=60
    Ribbon Width Mode=::NewEnumerator1
    UsePositionOffset=false
    Write Material Random=true
   )
   &PartitionParticles()
   &SkeletalMeshLocation(
    Bone / Socket Index=$Particles.PartitionParticles.Partition
    Bone Sampling Mode=::NewEnumerator4
    MaskSpawnGroup=false
    Orientation Sampling=::NewEnumerator1
    Position Sampling=::NewEnumerator1
    Skeletal Mesh=$System.SkeletalMeshUser
   )
   /SetVariables(Particles.Position=$Output.SkeletalMeshLocation.SampledPositionWithoutOffset) {isSetParameters=true, module=SetVariables_#2}
   &AddVelocity(
    Rotation Coordinate Space=::Simulation
    Velocity=&ReturnMeshOrientationAxis(Mesh Orientation Axis=::XAxis, Mesh Orientation Quat=$Particles.SkeletalMeshLocation.SampledRotation)
    Velocity Speed Scale=&MergeHelper_FloatSwitch(0=-300, 1=300, 2=?, 3=?, Index=$Particles.PartitionParticles.Partition, Num=::NewEnumerator0)
   )
   &ApplyInitialForces()
  ]
  particleUpdateScript [
   &ParticleState()
   &ScaleColor(
    Scale Alpha=&FloatFromCurve(
     FloatCurve=DataInterface() {propertyValues {Curve {
        defaultValue=1
        keys [
         {interpMode=RCIM_Cubic, tangentMode=RCTM_User, time=0, value=1}
         {arriveTangent=-2.1334, interpMode=RCIM_Cubic, leaveTangent=-2.1334, time=0.3242, value=1}
         <at 0.4687 0 RCIM_Cubic>
        ]
       }}}
     Scale Curve=&Multiply_Float(A=10, B=1)
    )
    Scale RGB=&MakeVectorFromLinearColorRGB(Color=<rgba 1 1 1 1>)
    ScaleA=true
    ScaleRGB=true
   )
   /SetVariables(Particles.GlowColor=&MultiplyLinearColor(
     Linear Color=<rgba 0.3391 0.9196 1.5 1>
     Scale Factor=&MakeVector4FromVectorAndFloat(
      "Vector (RGB)"=&MakeVectorFromLinearColorRGB(Color=$System.TODColor)
      "Float (Alpha)"=&FloatFromCurve(
       FloatCurve=DataInterface() {propertyValues {Curve {
          defaultValue=1
          keys [
           {interpMode=RCIM_Cubic, tangentMode=RCTM_User, time=0, value=1}
           {arriveTangent=-1.0589, interpMode=RCIM_Cubic, leaveTangent=-1.0589, tangentMode=RCTM_User, time=1, value=0}
          ]
         }}}
       Scale Curve=$System.SwingGeneralFade
      )
     )
    )) {isSetParameters=true, module=SetVariables_#4}
   &ScaleRibbonWidth(Ribbon Width Scale=&Multiply_Float(
     A=&FloatFromCurve(FloatCurve=DataInterface() {propertyValues {Curve {defaultValue=0, keys [{interpMode=RCIM_Cubic, tangentMode=RCTM_User, time=0, value=0.5}]}}})
     B=1
    ))
   &SolveForcesAndVelocity(Clamp Velocity=false, Limit Acceleration=false)
   &ScaleColor(Scale Alpha=0.6, ScaleA=true, ScaleRGB=true) {module=ScaleColor001}
  ]
  rendererClasses [&NiagaraRibbonRendererProperties]
  renderers [
   {rendererClass=&NiagaraRibbonRendererProperties}
   {rendererClass=&NiagaraRibbonRendererProperties, rendererIndex=1}
  ]
 }
 Toon_Fire004 {
  emitterUpdateScript [
   &EmitterState()
   &SpawnRate(SpawnRate=100, Use Spawn Probability=false)
  ]
  particleSpawnScript [
   &InitializeParticle(
    Color=<rgba 0 9.1556 20 1>
    Lifetime=0.6
    Ribbon Width=50
    Ribbon Width Mode=::NewEnumerator1
    UsePositionOffset=false
    Write Material Random=true
   )
   &PartitionParticles()
   &SkeletalMeshLocation(
    Bone / Socket Index=$Particles.PartitionParticles.Partition
    Bone Sampling Mode=::NewEnumerator4
    MaskSpawnGroup=false
    Orientation Sampling=::NewEnumerator1
    Position Sampling=::NewEnumerator1
    Skeletal Mesh=$System.SkeletalMeshUser
   )
   /SetVariables(Particles.Position=$Output.SkeletalMeshLocation.SampledPositionWithoutOffset) {isSetParameters=true, module=SetVariables_#2}
   &AddVelocity(
    Rotation Coordinate Space=::Simulation
    Velocity=&ReturnMeshOrientationAxis(Mesh Orientation Axis=::XAxis, Mesh Orientation Quat=$Particles.SkeletalMeshLocation.SampledRotation)
    Velocity Speed Scale=&MergeHelper_FloatSwitch(
     0=&Subtract_Float(A=0, B=$User.User_Velocity)
     1=$User.User_Velocity
     2=?
     3=?
     Index=$Particles.PartitionParticles.Partition
     Num=::NewEnumerator0
    )
   )
   &ApplyInitialForces()
  ]
  particleUpdateScript [
   &ParticleState()
   &ScaleColor(
    Scale Alpha=&FloatFromCurve(FloatCurve=DataInterface() {propertyValues {Curve {
        defaultValue=1
        keys [
         {interpMode=RCIM_Cubic, tangentMode=RCTM_User, time=0, value=0}
         <at 0.1 1 RCIM_Cubic>
         {interpMode=RCIM_Cubic, leaveTangent=10, time=0.2, value=0}
        ]
       }}})
    Scale RGB=&MakeVectorFromLinearColorRGB(Color=<rgba 1 1 1 1>)
    ScaleA=true
    ScaleRGB=true
   )
   /SetVariables(Particles.GlowColor=&MultiplyLinearColor(
     Linear Color=<rgba 0.3391 0.9196 1.5 1>
     Scale Factor=&MakeVector4FromVectorAndFloat(
      "Vector (RGB)"=&MakeVectorFromLinearColorRGB(Color=$System.TODColor)
      "Float (Alpha)"=&FloatFromCurve(
       FloatCurve=DataInterface() {propertyValues {Curve {
          defaultValue=1
          keys [
           {interpMode=RCIM_Cubic, tangentMode=RCTM_User, time=0, value=1}
           {arriveTangent=-1.0589, interpMode=RCIM_Cubic, leaveTangent=-1.0589, tangentMode=RCTM_User, time=1, value=0}
          ]
         }}}
       Scale Curve=$System.SwingGeneralFade
      )
     )
    )) {isSetParameters=true, module=SetVariables_#4}
   &ScaleRibbonWidth(Ribbon Width Scale=&Multiply_Float(
     A=&FloatFromCurve(FloatCurve=DataInterface() {propertyValues {Curve {
         defaultValue=0
         keys [
          {interpMode=RCIM_Cubic, tangentMode=RCTM_User, time=0, value=1}
          {arriveTangent=-1.5474, interpMode=RCIM_Cubic, leaveTangent=-1.5474, tangentMode=RCTM_User, time=1, value=0}
         ]
        }}})
     B=1
    ))
   &SolveForcesAndVelocity(Clamp Velocity=false, Limit Acceleration=false)
  ]
  rendererClasses [&NiagaraRibbonRendererProperties]
  renderers [{rendererClass=&NiagaraRibbonRendererProperties}]
 }
 Toon_Fire006 {
  emitterUpdateScript [
   &EmitterState()
   &SpawnRate(SpawnRate=100, Use Spawn Probability=false)
  ]
  particleSpawnScript [
   &InitializeParticle(Ribbon Width=100, Ribbon Width Mode=::NewEnumerator1, UsePositionOffset=false, Write Material Random=true)
   &PartitionParticles()
   &SkeletalMeshLocation(
    Bone / Socket Index=$Particles.PartitionParticles.Partition
    Bone Sampling Mode=::NewEnumerator4
    MaskSpawnGroup=false
    Orientation Sampling=::NewEnumerator1
    Position Sampling=::NewEnumerator1
    Skeletal Mesh=$System.SkeletalMeshUser
   )
   /SetVariables(Particles.Position=$Output.SkeletalMeshLocation.SampledPositionWithoutOffset) {isSetParameters=true, module=SetVariables_#2}
   &AddVelocity(
    Rotation Coordinate Space=::Simulation
    Velocity=&ReturnMeshOrientationAxis(Mesh Orientation Axis=::XAxis, Mesh Orientation Quat=$Particles.SkeletalMeshLocation.SampledRotation)
    Velocity Speed Scale=&MergeHelper_FloatSwitch(
     0=&Subtract_Float(A=0, B=$User.User_Velocity)
     1=$User.User_Velocity
     2=?
     3=?
     Index=$Particles.PartitionParticles.Partition
     Num=::NewEnumerator0
    )
   )
   &ApplyInitialForces()
  ]
  particleUpdateScript [
   &ParticleState()
   &ScaleColor(
    Linear Color Curve=DataInterface() {propertyValues {
      AlphaCurve {keys [<at 0.0207 0>,<at 0.1253 0.9793>,<at 0.1846 0>,<at 1 0>]}
      BlueCurve {keys [<at 0.003 0.3229>,<at 0.0491 0.4507>,<at 0.1611 0.3177>,<at 0.3437 2>]}
      GreenCurve {keys [<at 0.003 0.8537>,<at 0.0491 0.2376>,<at 0.1611 0.0757>,<at 0.3437 0.5232>]}
      RedCurve {keys [<at 0.003 1>,<at 0.0491 1>,<at 0.1611 0.1416>,<at 0.3437 0.0938>]}
     }}
    Scale Mode=2
   )
   &ScaleRibbonWidth(Ribbon Width Scale=&Multiply_Float(
     A=&FloatFromCurve(FloatCurve=DataInterface() {propertyValues {Curve {
         defaultValue=0
         keys [
          {interpMode=RCIM_Cubic, tangentMode=RCTM_User, time=0, value=1}
          {arriveTangent=-1.5474, interpMode=RCIM_Cubic, leaveTangent=-1.5474, tangentMode=RCTM_User, time=1, value=0}
         ]
        }}})
     B=1
    ))
   &Drag(
    Drag=&FloatFromCurve(FloatCurve=DataInterface() {propertyValues {Curve {keys [<at 0 0>,<at 0.2138 1>,<at 1 1>]}}})
    Use Linear Drag=true
    Use Rotational Drag=true
   )
   &GravityForce(Gravity=&Lerp_Vector(A=<0 0 450>, B=<0 0 -450>, Alpha=$User.C_GravityMod))
   &GravityForce(
    Coordinate Space=::Simulation
    Gravity=&Lerp_Vector(
     A=<0 0 0>
     B=&AddVectorToPosition(
      Position=&MergeHelper_VectorSwitch(
       0=&Multiply_VectorByFloat(
        Vector=&ReturnMeshOrientationAxis(Mesh Orientation Axis=::XAxis, Mesh Orientation Quat=$Particles.SkeletalMeshLocation.SampledRotation)
        Float=-300
       )
       1=&Multiply_VectorByFloat(
        Vector=&ReturnMeshOrientationAxis(Mesh Orientation Axis=::XAxis, Mesh Orientation Quat=$Particles.SkeletalMeshLocation.SampledRotation)
        Float=300
       )
       2=?
       3=?
       Index=$Particles.PartitionParticles.Partition
       Num=::NewEnumerator0
      )
      Vector=<0 0 1>
     )
     Alpha=$User.C_GravityForce
    )
   ) {module=GravityForce001}
   &SolveForcesAndVelocity(Clamp Velocity=false, Limit Acceleration=false)
   &ScaleColor(Scale Alpha=1.5, ScaleA=true, ScaleRGB=true) {module=ScaleColor001}
  ]
  rendererClasses [&NiagaraRibbonRendererProperties]
  renderers [{rendererClass=&NiagaraRibbonRendererProperties}]
 }
 Toon_Water {
  emitterUpdateScript [
   &EmitterState()
   &SpawnPerFrame(Use Spawn Probability=false)
  ]
  particleSpawnScript [
   &InitializeParticle(Lifetime=0.5, Ribbon Width=50, Ribbon Width Mode=::NewEnumerator1, UsePositionOffset=false, Write Material Random=true)
   &PartitionParticles()
   &SkeletalMeshLocation(
    Bone / Socket Index=$Particles.PartitionParticles.Partition
    Bone Sampling Mode=::NewEnumerator4
    MaskSpawnGroup=false
    Orientation Sampling=::NewEnumerator1
    Position Sampling=::NewEnumerator1
    Skeletal Mesh=$System.SkeletalMeshUser
   )
   /SetVariables(Particles.Position=$Output.SkeletalMeshLocation.SampledPositionWithoutOffset) {isSetParameters=true, module=SetVariables_#2}
   &AddVelocity(
    Rotation Coordinate Space=::Simulation
    Velocity=&ReturnMeshOrientationAxis(Mesh Orientation Axis=::XAxis, Mesh Orientation Quat=$Particles.SkeletalMeshLocation.SampledRotation)
    Velocity Speed Scale=&MergeHelper_FloatSwitch(
     0=&Subtract_Float(A=0, B=$User.User_Velocity)
     1=$User.User_Velocity
     2=?
     3=?
     Index=$Particles.PartitionParticles.Partition
     Num=::NewEnumerator0
    )
   )
   &ApplyInitialForces()
  ]
  particleUpdateScript [
   &ParticleState()
   &ScaleColor(
    Scale Alpha=&FloatFromCurve(FloatCurve=DataInterface() {propertyValues {Curve {
        defaultValue=1
        keys [
         {interpMode=RCIM_Cubic, tangentMode=RCTM_User, time=0, value=0}
         {arriveTangent=2, interpMode=RCIM_Cubic, leaveTangent=2, time=0.2, value=0}
         <at 0.5 1 RCIM_Cubic>
         <at 1 0 RCIM_Cubic>
        ]
       }}})
    Scale RGB=&MakeVectorFromLinearColorRGB(Color=<rgba 1 1 1 1>)
    ScaleA=true
    ScaleRGB=true
   ) {module=ScaleColor001}
   &ScaleRibbonWidth(Ribbon Width Scale=&Multiply_Float(
     A=&FloatFromCurve(FloatCurve=DataInterface() {propertyValues {Curve {
         defaultValue=0
         keys [
          {arriveTangent=-3.7011, interpMode=RCIM_Cubic, time=0, value=0}
          {arriveTangent=0.562, interpMode=RCIM_Cubic, leaveTangent=0.562, tangentMode=RCTM_User, time=0.3, value=1.1103}
          {interpMode=RCIM_Cubic, leaveTangent=0.3678, time=0.6, value=0}
         ]
        }}})
     B=1
    ))
   &Drag(
    Drag=&FloatFromCurve(FloatCurve=DataInterface() {propertyValues {Curve {keys [<at 0 0>,<at 0.2 1>,<at 1 1>]}}})
    Use Linear Drag=true
    Use Rotational Drag=true
   ) {enabled=false}
   &GravityForce(Gravity=&Lerp_Vector(A=<0 0 450>, B=<0 0 -450>, Alpha=$User.C_GravityMod))
   &GravityForce(
    Coordinate Space=::Simulation
    Gravity=&Lerp_Vector(
     A=<0 0 0>
     B=&AddVectorToPosition(
      Position=&MergeHelper_VectorSwitch(
       0=&Multiply_VectorByFloat(
        Vector=&ReturnMeshOrientationAxis(Mesh Orientation Axis=::XAxis, Mesh Orientation Quat=$Particles.SkeletalMeshLocation.SampledRotation)
        Float=-300
       )
       1=&Multiply_VectorByFloat(
        Vector=&ReturnMeshOrientationAxis(Mesh Orientation Axis=::XAxis, Mesh Orientation Quat=$Particles.SkeletalMeshLocation.SampledRotation)
        Float=300
       )
       2=?
       3=?
       Index=$Particles.PartitionParticles.Partition
       Num=::NewEnumerator0
      )
      Vector=<0 0 1>
     )
     Alpha=$User.C_GravityForce
    )
   ) {module=GravityForce001}
   &SolveForcesAndVelocity(Clamp Velocity=false, Limit Acceleration=false)
   &ScaleColor(Scale Alpha=0.3, ScaleA=true, ScaleRGB=true)
  ]
  rendererClasses [&NiagaraRibbonRendererProperties]
  renderers [
   {rendererClass=&NiagaraRibbonRendererProperties}
   {rendererClass=&NiagaraRibbonRendererProperties, rendererIndex=1}
  ]
 }
 Toon_Water002 {
  emitterUpdateScript [
   &EmitterState()
   &SpawnPerFrame(Use Spawn Probability=false)
  ]
  particleSpawnScript [
   &InitializeParticle(Lifetime=0.5, Ribbon Width=50, Ribbon Width Mode=::NewEnumerator1, UsePositionOffset=false, Write Material Random=true)
   &PartitionParticles()
   &SkeletalMeshLocation(
    Bone / Socket Index=$Particles.PartitionParticles.Partition
    Bone Sampling Mode=::NewEnumerator4
    MaskSpawnGroup=false
    Orientation Sampling=::NewEnumerator1
    Position Sampling=::NewEnumerator1
    Skeletal Mesh=$System.SkeletalMeshUser
   )
   /SetVariables(Particles.Position=$Output.SkeletalMeshLocation.SampledPositionWithoutOffset) {isSetParameters=true, module=SetVariables_#2}
   &AddVelocity(
    Rotation Coordinate Space=::Simulation
    Velocity=&ReturnMeshOrientationAxis(Mesh Orientation Axis=::XAxis, Mesh Orientation Quat=$Particles.SkeletalMeshLocation.SampledRotation)
    Velocity Speed Scale=&MergeHelper_FloatSwitch(
     0=&Subtract_Float(A=0, B=$User.User_Velocity)
     1=$User.User_Velocity
     2=?
     3=?
     Index=$Particles.PartitionParticles.Partition
     Num=::NewEnumerator0
    )
   )
   &ApplyInitialForces()
  ]
  particleUpdateScript [
   &ParticleState()
   &ScaleColor(
    Scale Alpha=&FloatFromCurve(FloatCurve=DataInterface() {propertyValues {Curve {
        defaultValue=1
        keys [
         {interpMode=RCIM_Cubic, tangentMode=RCTM_User, time=0, value=0}
         {arriveTangent=3.3333, interpMode=RCIM_Cubic, leaveTangent=3.3333, time=0.2, value=0}
         <at 0.3 1 RCIM_Cubic>
         <at 0.5 0 RCIM_Cubic>
        ]
       }}})
    Scale RGB=&MakeVectorFromLinearColorRGB(Color=<rgba 1 1 1 1>)
    ScaleA=true
    ScaleRGB=true
   ) {module=ScaleColor001}
   &ScaleRibbonWidth(Ribbon Width Scale=&Multiply_Float(
     A=&FloatFromCurve(FloatCurve=DataInterface() {propertyValues {Curve {
         defaultValue=0
         keys [
          {arriveTangent=-3.7011, interpMode=RCIM_Cubic, time=0, value=0}
          {arriveTangent=0.562, interpMode=RCIM_Cubic, leaveTangent=0.562, tangentMode=RCTM_User, time=0.3, value=1.1103}
          {interpMode=RCIM_Cubic, leaveTangent=0.3678, time=0.7, value=0}
         ]
        }}})
     B=1
    ))
   &Drag(
    Drag=&FloatFromCurve(FloatCurve=DataInterface() {propertyValues {Curve {keys [<at 0 0>,<at 0.2138 1>,<at 1 1>]}}})
    Use Linear Drag=true
    Use Rotational Drag=true
   )
   &GravityForce(Gravity=&Lerp_Vector(A=<0 0 450>, B=<0 0 -450>, Alpha=$User.C_GravityMod))
   &GravityForce(
    Coordinate Space=::Simulation
    Gravity=&Lerp_Vector(
     A=<0 0 0>
     B=&AddVectorToPosition(
      Position=&MergeHelper_VectorSwitch(
       0=&Multiply_VectorByFloat(
        Vector=&ReturnMeshOrientationAxis(Mesh Orientation Axis=::XAxis, Mesh Orientation Quat=$Particles.SkeletalMeshLocation.SampledRotation)
        Float=-300
       )
       1=&Multiply_VectorByFloat(
        Vector=&ReturnMeshOrientationAxis(Mesh Orientation Axis=::XAxis, Mesh Orientation Quat=$Particles.SkeletalMeshLocation.SampledRotation)
        Float=300
       )
       2=?
       3=?
       Index=$Particles.PartitionParticles.Partition
       Num=::NewEnumerator0
      )
      Vector=<0 0 1>
     )
     Alpha=$User.C_GravityForce
    )
   ) {module=GravityForce001}
   &SolveForcesAndVelocity(Clamp Velocity=false, Limit Acceleration=false)
   &ScaleColor(ScaleA=true, ScaleRGB=true)
  ]
  rendererClasses [&NiagaraRibbonRendererProperties]
  renderers [
   {rendererClass=&NiagaraRibbonRendererProperties}
   {rendererClass=&NiagaraRibbonRendererProperties, rendererIndex=1}
  ]
 }
}
summary {
 systemName=NS_Emote_FlowingFlames_Pre_Combine
 emitters [
  {emitterName=Collision_Position, rendererClasses [&NiagaraSpriteRendererProperties]}
  {emitterName=Toon_Fire, rendererClasses [&NiagaraRibbonRendererProperties]}
  {emitterName=SplashSprites, rendererClasses [&NiagaraSpriteRendererProperties]}
  {bEnabled=false, emitterName=Toon_Fire002, rendererClasses [&NiagaraRibbonRendererProperties]}
  {emitterName=Toon_Fire004, rendererClasses [&NiagaraRibbonRendererProperties]}
  {emitterName=FireSprite001, rendererClasses [&NiagaraSpriteRendererProperties]}
  {emitterName=Toon_Fire006, rendererClasses [&NiagaraRibbonRendererProperties]}
  {emitterName=Toon_Water002, rendererClasses [&NiagaraRibbonRendererProperties]}
  {emitterName=NE_Sparks, simTarget=GPUComputeSim, rendererClasses [&NiagaraSpriteRendererProperties]}
  {emitterName=SplashSprites001, rendererClasses [&NiagaraSpriteRendererProperties]}
  {emitterName=SplashSprites003, rendererClasses [&NiagaraSpriteRendererProperties]}
  {emitterName=Toon_Water, rendererClasses [&NiagaraRibbonRendererProperties]}
  {emitterName=Comet_Flame_Mesh, rendererClasses [&NiagaraMeshRendererProperties]}
  {emitterName=ConeLights, rendererClasses [&NiagaraLightRendererProperties]}
  {emitterName=SplashSprites002, rendererClasses [&NiagaraSpriteRendererProperties]}
  {bEnabled=false, emitterName=Fire_Water_Trail, rendererClasses [&NiagaraRibbonRendererProperties]}
 ]
 userVariables [
  User.bIsFrontEnd {type=NiagaraBool}
  User.bIsFrontEndPreview {type=NiagaraBool}
  User.C_Anti_Popcorn {type=NiagaraFloat, defaultValue=0, description=""}
  User.C_GravityForce {type=NiagaraFloat, defaultValue=0, description=""}
  User.C_GravityMod {type=NiagaraFloat, defaultValue=0, description=""}
  User.C_Inherit_Velocity {type=NiagaraFloat, defaultValue=0, description=""}
  User.C_Steam_Opacity {type=NiagaraFloat, defaultValue=0, description=""}
  User.Flame_Size {type=NiagaraFloat, defaultValue=1, description=""}
  User.RibbonIDOffset {type=NiagaraInt32}
  User.User_Velocity {type=NiagaraFloat, defaultValue=175, description=""}
 ]
}
